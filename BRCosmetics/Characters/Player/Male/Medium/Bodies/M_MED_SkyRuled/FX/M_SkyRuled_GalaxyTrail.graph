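# Unlit translucent galaxy trail material for the SkyRuled character cosmetic, built for Niagara ribbons and mesh particles.
# Emissive tints screen-aligned nebula and cell-noise-broken stars with a curve-atlas gradient along U and particle color.
# Opacity combines panning smoke and fire masks, flame dissolve, edge, lead/tail, view-angle and depth fades, plus an optional Appear wipe.
# Dynamic parameters drive appearance, time offset and SplineThicken width, which a scalar blends against camera offset.
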Material /BRCosmetics/Characters/Player/Male/Medium/Bodies/M_MED_SkyRuled/FX/M_SkyRuled_GalaxyTrail
graph {
 properties {
  bAutomaticallySetUsageInEditor=false
  bUsedWithNiagaraMeshParticles=true
  bUsedWithNiagaraRibbons=true
  blendMode=BLEND_Translucent
  shadingModel=MSM_Unlit
  twoSided=true
  usedShadingModels=Unlit
 }
 nodes [
  n1 = TextureCoordinate()
  n2 = Multiply(A=Sine(None=ComponentMask(None=@n1) {r=true}) {period=2}) {constB=6}
  n3 = Multiply(
   A=DynamicParameter() {defaultValue=<rgba 10 1 0 1>, paramNames [SplineThickenWidth, WidthCurve, Appear, Timeoffset]}.SplineThickenWidth
   B=LinearInterpolate(Alpha=Saturate(None=@n2)) {constA=0.3}
  )
  LinearInterpolate(
   A=MaterialFunctionCall(Offset Amount=Constant() {r=1}) {materialFunction=/Engine/Functions/Engine_MaterialFunctions02/WorldPositionOffset/CameraOffset}.Camera Offset
   B=MaterialFunctionCall(WidthBase=@n3, WidthTip=@n3, Expand U or V UV Channel=StaticBool()) {materialFunction=&SplineThicken}.WorldPositionOffset
   Alpha=ScalarParameter() {parameterName=UseSplineThicken}
  )
  n4 = CurveAtlasRowParameter(CurveTime=ComponentMask(None=TextureCoordinate()) {r=true}) {
   atlas=&/CA_SkyRuled_FXColor
   curve=&/CV_SkyRuled_FXColor01
   parameterName=ColorGradient
  }
  n5 = TextureSample(UVs=MaterialFunctionCall!(TextureResolution=ScalarParameter() {defaultValue=100, parameterName=Stars Size}) {materialFunction=/Engine/Functions/Engine_MaterialFunctions02/Texturing/ScreenAlignedPixelToPixelUVs}.Result) {samplerType=SAMPLERTYPE_Masks, texture=/Game/Characters/Player/Male/Medium/Bodies/M_Med_Soldier_04/Skins/Galaxy/Textures/Celestial_T_Stars}
  n6 = ComponentMask(None=@n5.RGB) {b=true}
  n7 = Add(B=Multiply(A=Power(Base=@n6) {constExponent=1}, B=ScalarParameter() {defaultValue=100, parameterName=Star_Mul}) {constB=100})
  n8 = TextureSample(UVs=TextureCoordinate() {uTiling=2}) {samplerType=SAMPLERTYPE_Grayscale, texture=/BRCosmetics/Weapons/FORT_Melee/Pickaxe_Cube_Ninja_Male/FX/T_CellNoise}
  n9 = TextureSampleParameter2D(UVs=MaterialFunctionCall(TextureResolution=ScalarParameter() {defaultValue=256, parameterName=Galaxy Size}) {materialFunction=&ScreenAlignedPixelToPixelUVs}.Result) {
   parameterName=GalaxyTexture
   texture=/Game/Characters/Player/Female/Medium/Bodies/F_MED_SuperNovaTaro/FX/Material/T_SuperNovaTaro_Skin_Nebula
  }
  n10 = Add(A=Add(B=Multiply(A=@n7, B=@n8.R)), B=Multiply(A=@n9.RGB) {constB=6}) {constB=0}
  n11 = ComponentMask(None=TextureCoordinate()) {g=true}
  n12 = LinearInterpolate(Alpha=Power(Base=Sine(None=@n11) {period=2}) {constExponent=1}) {constA=0.2}
  n13 = Multiply(
   A=@n4
   B=LinearInterpolate(A=LinearInterpolate(B=@n10, Alpha=@n12) {constA=1}, Alpha=ScalarParameter() {defaultValue=0.25, parameterName=GalaxyLerp}) {constAlpha=0.3}
  )
  n14 = ScalarParameter() {defaultValue=-1, parameterName=Des}
  n15 = ParticleColor()
  MaterialFunctionCall(Emissive=Multiply(A=Desaturation(None=@n13, Fraction=@n14), B=@n15.RGB)) {materialFunction=&MF_FX_EmissiveAdjustments}
  n16 = ScalarParameter() {defaultValue=3, parameterName=FadeDistance}
  n17 = TextureCoordinate()
  n18 = Power(Base=Sine(None=ComponentMask(None=@n17) {g=true}) {period=2}) {constExponent=1}
  n19 = ComponentMask(None=VectorParameter() {parameterName=UVScale, defaultValue=<rgba 1 1 0 0>}.RGB) {g=true, r=true}
  n20 = Add(
   A=TextureCoordinate() {uTiling=1.5}
   B=AppendVector(
    A=DynamicParameter() {defaultValue=<rgba 10 1 0 1>, paramNames [SplineThickenWidth, WidthCurve, Appear, Timeoffset]}.Timeoffset
    B=Constant()
   )
  )
  n21 = Multiply(A=Time(), B=ScalarParameter() {defaultValue=1, parameterName=Speed})
  n22 = TextureSampleParameter2D(UVs=Panner(Coordinate=Multiply(A=@n20, B=@n19), Time=@n21) {speedX=-0.6}) {
   parameterName=MaskTexture01
   samplerType=SAMPLERTYPE_LinearGrayscale
   texture=/BRCosmetics/Athena/Effects/Particles/Skydiving/NS_Contrail_LightCycle/Textures/T_FireTrail_01
  }
  n23 = ScalarParameter() {defaultValue=2, parameterName=MaskTexture02_UTiling}
  n24 = ScalarParameter() {defaultValue=0.7, parameterName=MaskTexture02_VTiling}
  n25 = Add(
   A=Multiply(A=TextureCoordinate(), B=AppendVector(A=@n23, B=@n24))
   B=AppendVector(
    A=DynamicParameter() {defaultValue=<rgba 10 1 0 1>, paramNames [SplineThickenWidth, WidthCurve, Appear, Timeoffset]}.Timeoffset
    B=Constant()
   )
  )
  n26 = TextureSampleParameter2D(UVs=Panner(Coordinate=Multiply(A=@n19, B=@n25), Time=@n21) {speedX=-0.3}) {parameterName=MaskTexture02, samplerType=SAMPLERTYPE_Grayscale, texture=&T_Smoke_Trail_Soft}
  n27 = Power(Base=LinearInterpolate(A=@n22.R, B=@n26.R) {constAlpha=0.6})
  n28 = TextureCoordinate()
  n29 = Multiply(A=Sine(None=ComponentMask(None=@n28) {r=true}) {period=2}) {constB=3}
  n30 = Multiply(A=@n15.A, B=Multiply(A=Multiply(A=@n18, B=@n27), B=Saturate(None=@n29)))
  n31 = MaterialFunctionCall("FadeSensitivity (optional)"=ScalarParameter() {defaultValue=1, parameterName=AngleFade}) {materialFunction=&FadeBasedOnViewAngle}
  n32 = TextureCoordinate() {vTiling=0.4}
  n33 = ComponentMask(None=VectorParameter() {parameterName=UVScale, defaultValue=<rgba 1 1 0 0>}.RGB) {g=true, r=true}
  n34 = DynamicParameter() {defaultValue=<rgba 10 1 0 1>, paramNames [SplineThickenWidth, WidthCurve, Appear, Timeoffset]}
  n35 = Constant()
  n36 = Panner(Coordinate=Add(A=Multiply(A=@n32, B=@n33), B=AppendVector(A=@n34.Timeoffset, B=@n35))) {speedX=-0.5}
  n37 = LinearInterpolate(A=ScalarParameter() {defaultValue=-2, parameterName=DissolveMin}, Alpha=ComponentMask(None=TextureCoordinate()) {r=true}) {constA=-2, constB=1.5}
  n38 = LinearInterpolate(Alpha=Power(Base=TextureSample(UVs=@n36) {texture=/Game/Effects/Fort_Effects/Textures/Tile/T_Tile_WaveFlame}.R, Exp=Saturate(None=@n37)))
  n39 = TextureCoordinate()
  n40 = ScalarParameter() {defaultValue=10, parameterName=LinearMask_Mul}
  n41 = Multiply(A=Subtract(A=ComponentMask(None=@n39) {r=true}, B=ScalarParameter() {defaultValue=0.13, parameterName=LeadRange}) {constB=0.13}, B=@n40) {constB=10}
  n42 = ComponentMask(None=TextureCoordinate()) {r=true}
  n43 = Multiply(A=Subtract(A=OneMinus(None=@n42), B=ScalarParameter() {defaultValue=0.2, parameterName=TailRange}) {constB=0.25}, B=@n40) {constB=10}
  n44 = Multiply(A=Multiply(A=Multiply(A=@n30, B=@n31.Result), B=Saturate(None=@n38)), B=Multiply(A=Saturate(None=@n41), B=Saturate(None=@n43)))
  n45 = OneMinus(None=ComponentMask(None=TextureCoordinate()) {r=true})
  n46 = LinearInterpolate(Alpha=DynamicParameter() {defaultValue=<rgba 10 1 0 1>, paramNames [SplineThickenWidth, WidthCurve, Appear, Timeoffset]}.Appear) {constA=-1}
  n47 = Saturate(None=Multiply(A=Add(A=@n45, B=@n46)) {constB=7})
  n48 = ScalarParameter() {parameterName=UseAppear}
  Reroute(None=Multiply(A=DepthFade(Opacity=@n44, FadeDistance=@n16), B=LinearInterpolate(B=@n47, Alpha=@n48) {constA=1}))
  VectorParameter() {parameterName=DissolveUV, defaultValue=<rgba 1 1 0 0>}
 ]
}
