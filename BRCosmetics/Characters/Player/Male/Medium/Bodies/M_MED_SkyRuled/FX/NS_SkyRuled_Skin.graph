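# Niagara cosmetic effect for the SkyRuled player body skin, adding flowing ribbons to both arms and fading mesh particles at the hands.
# Four arm emitters (large and small per side) burst persistent ordered points, placed by FN_BoneFlow from skeletal mesh readers with start, mid and end offsets.
# Each draws two ribbon layers with curve-tapered width and System.LODFade-scaled color.
# Hand spawning stops when User.bisHandHidden is set; the LinkStableTest emitter is disabled.
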NiagaraSystem /BRCosmetics/Characters/Player/Male/Medium/Bodies/M_MED_SkyRuled/FX/NS_SkyRuled_Skin
dependencies {
 usedDataInterfaces [
  &NiagaraDataInterfaceCurve
  &NiagaraDataInterfacePlatformSet
  &NiagaraDataInterfaceArrayFloat3
  &NiagaraDataInterfaceColorCurve
  &NiagaraDataInterfaceSkeletalMesh
 ]
 usedDynamicInputs [&FN_LOD_Distance_Remap]
 usedModules [
  &self:SystemSpawnScript.SystemScriptSource.SystemScriptGraph.NiagaraNodeAssignment_0.SetVariables_#1
  &self:SystemSpawnScript.SystemScriptSource.SystemScriptGraph.NiagaraNodeAssignment_1.SetVariables_#2
  &self:SystemSpawnScript.SystemScriptSource.SystemScriptGraph.NiagaraNodeAssignment_3.SetVariables_#3
  &SystemState
  &self:SystemSpawnScript.SystemScriptSource.SystemScriptGraph.NiagaraNodeAssignment_4.SetVariables_#4
 ]
}
emitters {
 ArmL {
  emitterUpdateScript [
   &EmitterState(Life Cycle Mode=::NewEnumerator1, Loop Behavior=::NewEnumerator1, Loop Duration Mode=::NewEnumerator1, UseLoopDelay=false)
   &SpawnBurst_Instantaneous(Use Spawn Probability=false, UseLoopCountLimit=false) {module=SpawnBurst_Instantaneous003}
   &SpawnBurst_Instantaneous(Spawn Group=1, Use Spawn Probability=false, UseLoopCountLimit=false) {module=SpawnBurst_Instantaneous002}
   &SpawnBurst_Instantaneous(Spawn Group=2, Use Spawn Probability=false, UseLoopCountLimit=false) {module=SpawnBurst_Instantaneous001}
   &SpawnBurst_Instantaneous(Spawn Group=3, Use Spawn Probability=false, UseLoopCountLimit=false)
  ]
  particleSpawnScript [
   &PartitionParticles(ID to Partition:NiagaraInt32=&SpawnGroup(), Number Of Partitions=3)
   &InitializeParticle(
    Color=<rgba 2 1.5312 1.7849 2>
    Ribbon Width=14
    Ribbon Width Mode=::NewEnumerator1
    Sprite Size Mode=::NewEnumerator3
    Uniform Sprite Size=20
    UsePositionOffset=false
    Write Material Random=true
   )
   /SetVariables(Particles.RibbonLinkOrder:NiagaraFloat=&Divide_Float(A=&MakeFloatFromInt(Int=&SpawnGroup()), B=3)) {isSetParameters=true, module=SetVariables_#5}
   &DynamicMaterialParameters(Module.Index0Float0Write=false, Module.Index0Float1Write=false, Module.Index0Float2Write=false, Module.Index0Float3Write=true)
  ]
  particleUpdateScript [
   &ParticleState(Kill Particles When Lifetime Has Elapsed=false)
   &FN_BoneFlow(
    Angle:NiagaraFloat=&Multiply_Float(A=&Multiply_Float(A=$Particles.PartitionParticles.NormalizedPartition, B=1), B=1)
    BoneID:NiagaraInt32=&ModuloInt(A=$Particles.PartitionParticles.Partition, B=1)
    Enable Position Offsets=true
    Height:NiagaraFloat=&Multiply_Float(A=$Particles.PartitionParticles.NormalizedPartition, B=1)
    Offset End:Vector3f=<-30 -50 0>
    Offset Mid:Vector3f=<35 5 20>
    Offset Start:Vector3f=<6 8 0>
    SkeletalMesh:NiagaraDataInterfaceSkeletalMesh=$System.Skeletal Mesh ReaderL
   )
   &ScaleRibbonWidth(Ribbon Width Scale=&FloatFromCurve(
     FloatCurve=DataInterface() {propertyValues {Curve {keys [
         <at -0.013 0 RCIM_Cubic>
         {arriveTangent=0.6934, interpMode=RCIM_Cubic, leaveTangent=0.6934, time=0.0459, value=0.7529}
         {arriveTangent=-0.6133, interpMode=RCIM_Cubic, leaveTangent=-0.6133, time=0.7176, value=0.5066}
         <at 0.9645 0.1895>
        ]}}}
     CurveIndex=$Particles.RibbonLinkOrder
     Scale Curve=&FN_PickFloatDependOnPlatform(TargetPlatformActive=1, TargetPlatformNoActive=1)
    ))
   &ScaleColor(Scale Alpha=&Multiply_Float(A=1.3, B=$System.LODFade), ScaleA=true, ScaleRGB=true)
   /SetVariables(Particles.RibbonWidth02:NiagaraFloat=&Multiply_Float(A=$Particles.RibbonWidth, B=1.5)) {isSetParameters=true}
   /SetVariables(Particles.Linear Color02:LinearColor=&Multiply_LinearColorByFloat(Linear Color=$Particles.Color, Float=&Multiply_Float(A=1.8, B=$System.LODFade))) {isSetParameters=true}
  ]
  rendererClasses [&NiagaraRibbonRendererProperties]
  renderers [
   {rendererClass=&NiagaraRibbonRendererProperties}
   {rendererClass=&NiagaraRibbonRendererProperties, rendererIndex=1}
  ]
 }
 ArmLSmall {
  emitterUpdateScript [
   &EmitterState(Life Cycle Mode=::NewEnumerator1, Loop Behavior=::NewEnumerator1, Loop Duration Mode=::NewEnumerator1, UseLoopDelay=false)
   &SpawnBurst_Instantaneous(Use Spawn Probability=false, UseLoopCountLimit=false) {module=SpawnBurst_Instantaneous003}
   &SpawnBurst_Instantaneous(Spawn Group=1, Use Spawn Probability=false, UseLoopCountLimit=false) {module=SpawnBurst_Instantaneous002}
   &SpawnBurst_Instantaneous(Spawn Group=2, Use Spawn Probability=false, UseLoopCountLimit=false) {module=SpawnBurst_Instantaneous001}
   &SpawnBurst_Instantaneous(Spawn Group=3, Use Spawn Probability=false, UseLoopCountLimit=false)
  ]
  particleSpawnScript [
   &PartitionParticles(ID to Partition:NiagaraInt32=&SpawnGroup(), Number Of Partitions=3)
   &InitializeParticle(
    Color=<rgba 2 1.5312 1.7849 2>
    Ribbon Width=8
    Ribbon Width Mode=::NewEnumerator1
    Sprite Size Mode=::NewEnumerator3
    Uniform Sprite Size=20
    UsePositionOffset=false
    Write Material Random=true
   )
   /SetVariables(Particles.RibbonLinkOrder:NiagaraFloat=&Divide_Float(A=&MakeFloatFromInt(Int=&SpawnGroup()), B=3)) {isSetParameters=true, module=SetVariables_#5}
   &DynamicMaterialParameters(
    Index 0 Param 4=0.7
    Module.Index0Float0Write=false
    Module.Index0Float1Write=false
    Module.Index0Float2Write=false
    Module.Index0Float3Write=true
   )
  ]
  particleUpdateScript [
   &ParticleState(Kill Particles When Lifetime Has Elapsed=false)
   &FN_BoneFlow(
    Angle:NiagaraFloat=&Multiply_Float(A=&Multiply_Float(A=$Particles.PartitionParticles.NormalizedPartition, B=1), B=1)
    BoneID:NiagaraInt32=&ModuloInt(A=$Particles.PartitionParticles.Partition, B=1)
    Enable Position Offsets=true
    Height:NiagaraFloat=&Multiply_Float(A=$Particles.PartitionParticles.NormalizedPartition, B=1)
    Offset End:Vector3f=<-30 -45 0>
    Offset Mid:Vector3f=<40 10 6>
    Offset Start:Vector3f=<5 5 0>
    SkeletalMesh:NiagaraDataInterfaceSkeletalMesh=$System.Skeletal Mesh ReaderL
   )
   &ScaleRibbonWidth(Ribbon Width Scale=&FloatFromCurve(
     FloatCurve=DataInterface() {propertyValues {Curve {keys [
         <at -0.013 0 RCIM_Cubic>
         {arriveTangent=0.6934, interpMode=RCIM_Cubic, leaveTangent=0.6934, time=0.0459, value=0.7529}
         {arriveTangent=-0.6133, interpMode=RCIM_Cubic, leaveTangent=-0.6133, time=0.7176, value=0.5066}
         <at 0.9645 0.1895>
        ]}}}
     CurveIndex=$Particles.RibbonLinkOrder
     Scale Curve=&FN_PickFloatDependOnPlatform(TargetPlatformActive=1, TargetPlatformNoActive=1)
    ))
   &ScaleColor(Scale Alpha=&Multiply_Float(A=1, B=$System.LODFade), ScaleA=true, ScaleRGB=true)
   /SetVariables(Particles.RibbonWidth02:NiagaraFloat=&Multiply_Float(A=$Particles.RibbonWidth, B=2)) {isSetParameters=true}
   /SetVariables(Particles.Linear Color02:LinearColor=&Multiply_LinearColorByFloat(Linear Color=$Particles.Color, Float=&Multiply_Float(A=1.8, B=$System.LODFade))) {isSetParameters=true}
  ]
  rendererClasses [&NiagaraRibbonRendererProperties]
  renderers [
   {rendererClass=&NiagaraRibbonRendererProperties}
   {rendererClass=&NiagaraRibbonRendererProperties, rendererIndex=1}
  ]
 }
 ArmR {
  emitterUpdateScript [
   &EmitterState(Life Cycle Mode=::NewEnumerator1, Loop Behavior=::NewEnumerator1, Loop Duration Mode=::NewEnumerator1, UseLoopDelay=false)
   &SpawnBurst_Instantaneous(Use Spawn Probability=false, UseLoopCountLimit=false) {module=SpawnBurst_Instantaneous003}
   &SpawnBurst_Instantaneous(Spawn Group=1, Use Spawn Probability=false, UseLoopCountLimit=false) {module=SpawnBurst_Instantaneous002}
   &SpawnBurst_Instantaneous(Spawn Group=2, Use Spawn Probability=false, UseLoopCountLimit=false) {module=SpawnBurst_Instantaneous001}
   &SpawnBurst_Instantaneous(Spawn Group=3, Use Spawn Probability=false, UseLoopCountLimit=false)
  ]
  particleSpawnScript [
   &PartitionParticles(ID to Partition:NiagaraInt32=&SpawnGroup(), Number Of Partitions=3)
   &InitializeParticle(
    Color=<rgba 2 1.5312 1.7849 2>
    Ribbon Width=14
    Ribbon Width Mode=::NewEnumerator1
    Sprite Size Mode=::NewEnumerator3
    Uniform Sprite Size=20
    UsePositionOffset=false
    Write Material Random=true
   )
   /SetVariables(Particles.RibbonLinkOrder:NiagaraFloat=&Divide_Float(A=&MakeFloatFromInt(Int=&SpawnGroup()), B=3)) {isSetParameters=true, module=SetVariables_#5}
   &DynamicMaterialParameters(
    Index 0 Param 4=0.5
    Module.Index0Float0Write=false
    Module.Index0Float1Write=false
    Module.Index0Float2Write=false
    Module.Index0Float3Write=true
   )
  ]
  particleUpdateScript [
   &ParticleState(Kill Particles When Lifetime Has Elapsed=false)
   &FN_BoneFlow(
    Angle:NiagaraFloat=&Multiply_Float(A=&Multiply_Float(A=$Particles.PartitionParticles.NormalizedPartition, B=1), B=1)
    BoneID:NiagaraInt32=&ModuloInt(A=$Particles.PartitionParticles.Partition, B=1)
    Enable Position Offsets=true
    Height:NiagaraFloat=&Multiply_Float(A=$Particles.PartitionParticles.NormalizedPartition, B=1)
    Offset End:Vector3f=<-40 -35 0>
    Offset Mid:Vector3f=<-35 0 -20>
    Offset Start:Vector3f=<3 11 0>
    SkeletalMesh:NiagaraDataInterfaceSkeletalMesh=$System.Skeletal Mesh ReaderR
   )
   &ScaleRibbonWidth(Ribbon Width Scale=&FloatFromCurve(
     FloatCurve=DataInterface() {propertyValues {Curve {keys [
         <at -0.013 0 RCIM_Cubic>
         {arriveTangent=0.6934, interpMode=RCIM_Cubic, leaveTangent=0.6934, time=0.0459, value=0.7529}
         {arriveTangent=-0.6133, interpMode=RCIM_Cubic, leaveTangent=-0.6133, time=0.7176, value=0.5066}
         <at 0.9645 0.1895>
        ]}}}
     CurveIndex=$Particles.RibbonLinkOrder
     Scale Curve=&FN_PickFloatDependOnPlatform(TargetPlatformActive=1, TargetPlatformNoActive=1)
    ))
   &ScaleColor(Scale Alpha=&Multiply_Float(A=1.3, B=$System.LODFade), ScaleA=true, ScaleRGB=true)
   /SetVariables(Particles.RibbonWidth02:NiagaraFloat=&Multiply_Float(A=$Particles.RibbonWidth, B=1.5)) {isSetParameters=true, module=SetVariables_#6}
   /SetVariables(Particles.Linear Color02:LinearColor=&Multiply_LinearColorByFloat(Linear Color=$Particles.Color, Float=&Multiply_Float(A=1.8, B=$System.LODFade))) {isSetParameters=true, module=SetVariables_#7}
  ]
  rendererClasses [&NiagaraRibbonRendererProperties]
  renderers [
   {rendererClass=&NiagaraRibbonRendererProperties}
   {rendererClass=&NiagaraRibbonRendererProperties, rendererIndex=1}
  ]
 }
 ArmR_Small {
  emitterUpdateScript [
   &EmitterState(Life Cycle Mode=::NewEnumerator1, Loop Behavior=::NewEnumerator1, Loop Duration Mode=::NewEnumerator1, UseLoopDelay=false)
   &SpawnBurst_Instantaneous(Use Spawn Probability=false, UseLoopCountLimit=false) {module=SpawnBurst_Instantaneous003}
   &SpawnBurst_Instantaneous(Spawn Group=1, Use Spawn Probability=false, UseLoopCountLimit=false) {module=SpawnBurst_Instantaneous002}
   &SpawnBurst_Instantaneous(Spawn Group=2, Use Spawn Probability=false, UseLoopCountLimit=false) {module=SpawnBurst_Instantaneous001}
   &SpawnBurst_Instantaneous(Spawn Group=3, Use Spawn Probability=false, UseLoopCountLimit=false)
  ]
  particleSpawnScript [
   &PartitionParticles(ID to Partition:NiagaraInt32=&SpawnGroup(), Number Of Partitions=3)
   &InitializeParticle(
    Color=<rgba 2 1.5312 1.7849 2>
    Ribbon Width=8
    Ribbon Width Mode=::NewEnumerator1
    Sprite Size Mode=::NewEnumerator3
    Uniform Sprite Size=20
    UsePositionOffset=false
    Write Material Random=true
   )
   /SetVariables(Particles.RibbonLinkOrder:NiagaraFloat=&Divide_Float(A=&MakeFloatFromInt(Int=&SpawnGroup()), B=3)) {isSetParameters=true, module=SetVariables_#5}
   &DynamicMaterialParameters(
    Index 0 Param 4=0.2
    Module.Index0Float0Write=false
    Module.Index0Float1Write=false
    Module.Index0Float2Write=false
    Module.Index0Float3Write=true
   )
  ]
  particleUpdateScript [
   &ParticleState(Kill Particles When Lifetime Has Elapsed=false)
   &FN_BoneFlow(
    Angle:NiagaraFloat=&Multiply_Float(A=&Multiply_Float(A=$Particles.PartitionParticles.NormalizedPartition, B=1), B=1)
    BoneID:NiagaraInt32=&ModuloInt(A=$Particles.PartitionParticles.Partition, B=1)
    Enable Position Offsets=true
    Height:NiagaraFloat=&Multiply_Float(A=$Particles.PartitionParticles.NormalizedPartition, B=1.1)
    Offset End:Vector3f=<-30 -35 -10>
    Offset Mid:Vector3f=<-40 -10 -5>
    Offset Start:Vector3f=<3 5 7>
    SkeletalMesh:NiagaraDataInterfaceSkeletalMesh=$System.Skeletal Mesh ReaderR
   )
   &ScaleRibbonWidth(Ribbon Width Scale=&FloatFromCurve(
     FloatCurve=DataInterface() {propertyValues {Curve {keys [
         <at -0.013 0 RCIM_Cubic>
         {arriveTangent=0.6934, interpMode=RCIM_Cubic, leaveTangent=0.6934, time=0.0459, value=0.7529}
         {arriveTangent=-0.6133, interpMode=RCIM_Cubic, leaveTangent=-0.6133, time=0.7176, value=0.5066}
         <at 0.9645 0.1895>
        ]}}}
     CurveIndex=$Particles.RibbonLinkOrder
     Scale Curve=&FN_PickFloatDependOnPlatform(TargetPlatformActive=1, TargetPlatformNoActive=1)
    ))
   &ScaleColor(Scale Alpha=&Multiply_Float(A=1, B=$System.LODFade), ScaleA=true, ScaleRGB=true)
   /SetVariables(Particles.RibbonWidth02:NiagaraFloat=&Multiply_Float(A=$Particles.RibbonWidth, B=2)) {isSetParameters=true}
   /SetVariables(Particles.Linear Color02:LinearColor=&Multiply_LinearColorByFloat(Linear Color=$Particles.Color, Float=&Multiply_Float(A=1.8, B=$System.LODFade))) {isSetParameters=true}
  ]
  rendererClasses [&NiagaraRibbonRendererProperties]
  renderers [
   {rendererClass=&NiagaraRibbonRendererProperties}
   {rendererClass=&NiagaraRibbonRendererProperties, rendererIndex=1}
  ]
 }
 Hand {
  emitterUpdateScript [
   &EmitterState(
    Life Cycle Mode=::NewEnumerator1
    Loop Delay=&RandomRangeFloat(Minimum=0, Maximum=2)
    Loop Duration=&RandomRangeFloat(Minimum=0.5, Maximum=1)
    Recalculate Duration Each Loop=true
    UseLoopDelay=true
   )
   &SpawnRate(
    Spawn Probability=&MakeCustomFloatFromBool(Bool=$User.bisHandHidden, True Float=0, False Float=1)
    SpawnRate=6
    Use Spawn Probability=true
   )
  ]
  particleSpawnScript [
   &PartitionParticles()
   &InitializeParticle(
    Color=<rgba 1 1 1 4>
    Lifetime Min=1.5
    Lifetime Mode=::NewEnumerator1
    Mesh Scale=&Multiply_VectorByFloat(Vector=<1 1 1>, Float=0.15)
    Mesh Scale Mode=::NewEnumerator4
    UsePositionOffset=false
    Write Material Random=true
   )
   &InitialMeshOrientation(
    Mesh Orientation Mode=::NewEnumerator2
    Rotation=&SelectVectorFromArray(
     Direct Array Index=$Particles.PartitionParticles.Partition
     Vector Selection Array=DataInterface() {propertyValues {FloatData [<0 -0.25 0>, <0 0.25 0>]}}
    )
    Use Rotation Vector=true
   )
   &InitialMeshOrientation(
    Mesh Orientation Mode=::NewEnumerator2
    Rotation=&RandomRangeVector(Minimum=<-1 0 0>, Maximum=<1 0 0>)
    Use Rotation Vector=true
   ) {module=InitialMeshOrientation001}
   &CameraOffset(Camera Offset Amount=10)
  ]
  particleUpdateScript [
   &ParticleState()
   &ScaleMeshSize(Scale Factor=&Multiply_VectorByFloat(Vector=<0.7 0.7 1>, Float=&RandomRangeFloat(Minimum=0.3, Maximum=0.4)))
   &SkeletalMeshLocation(
    Bone / Socket Index=$Particles.PartitionParticles.Partition
    Bone Sampling Mode=::NewEnumerator4
    MaskSpawnGroup=false
    Sampled Position Offset=&AddVectorToPosition(
     Position=<0 0 0>
     Vector=&Lerp_Vector(
      A=&SelectVectorFromArray(
       Direct Array Index=$Particles.PartitionParticles.Partition
       Vector Selection Array=DataInterface() {propertyValues {FloatData [<30 0 0>, <-30 0 0>]}}
      )
      B=&Multiply_VectorByFloat(
       Vector=&SelectVectorFromArray(
        Direct Array Index=$Particles.PartitionParticles.Partition
        Vector Selection Array=DataInterface() {propertyValues {FloatData [<20 0 0>, <-20 0 0>]}}
       )
       Float=&RandomRangeFloat(Minimum=0.4, Maximum=0.8)
      )
      Alpha=$Particles.NormalizedAge
     )
    )
    Skeletal Mesh=$System.Skeletal Mesh Hand
   )
   &ScaleColor(
    Scale Alpha=&FloatFromCurve(
     FloatCurve=DataInterface() {propertyValues {Curve {
        defaultValue=1
        keys [
         {interpMode=RCIM_Cubic, tangentMode=RCTM_User, time=0, value=0}
         {interpMode=RCIM_Cubic, tangentMode=RCTM_User, time=0.5, value=1}
         {tangentMode=RCTM_User, time=1, value=0}
        ]
       }}}
     Scale Curve=&FloatFromCurve(FloatCurve=DataInterface() {propertyValues {Curve {
         defaultValue=1
         keys [{interpMode=RCIM_Cubic, tangentMode=RCTM_User, time=0, value=1}, {interpMode=RCIM_Cubic, tangentMode=RCTM_User, time=1, value=0}]
        }}})
    )
    Scale RGB=&MakeVectorFromLinearColorRGB(Color=&ColorFromCurve(ColorCurve=DataInterface() {propertyValues {
        AlphaCurve {keys []}
        BlueCurve {keys [<at 0 1 RCIM_Cubic>,<at 0.6998 1 RCIM_Cubic>]}
        GreenCurve {keys [<at 0 1 RCIM_Cubic>,<at 0.6998 1 RCIM_Cubic>]}
        RedCurve {keys [<at 0 1 RCIM_Cubic>,<at 0.6998 1 RCIM_Cubic>]}
       }}))
    ScaleA=true
    ScaleRGB=true
   )
   &DynamicMaterialParameters(
    Index 0 Param 1=&Multiply_Float(
     A=&RandomRangeFloat(Minimum=5, Maximum=8)
     B=&FloatFromCurve(FloatCurve=DataInterface() {propertyValues {Curve {keys [<at 0 1 RCIM_Cubic>, {arriveTangent=-0.6508, interpMode=RCIM_Cubic, leaveTangent=-0.6508, tangentMode=RCTM_User, time=1, value=0.3}]}}})
    )
    Index 0 Param 2=&FloatFromCurve(FloatCurve=DataInterface() {propertyValues {Curve {keys [<at 0 0>, {arriveTangent=1, interpMode=RCIM_Cubic, leaveTangent=1, time=0.4393, value=0.8}, <at 1 1>]}}})
    Index 0 Param 3=&FloatFromCurve(FloatCurve=DataInterface() {propertyValues {Curve {defaultValue=1, keys [{tangentMode=RCTM_User, time=0, value=0}, {tangentMode=RCTM_User, time=1, value=1}]}}})
    Module.Index0Float0Write=true
    Module.Index0Float1Write=false
    Module.Index0Float2Write=true
    Module.Index0Float3Write=true
   )
   &SolveForcesAndVelocity(Clamp Velocity=false, Limit Acceleration=false)
  ]
  rendererClasses [&NiagaraMeshRendererProperties]
  renderers [{rendererClass=&NiagaraMeshRendererProperties, rendererIndex=0}]
 }
 LinkStableTest {
  bEnabled=false
  emitterUpdateScript [
   &EmitterState(Life Cycle Mode=::NewEnumerator1, Loop Behavior=::NewEnumerator1, Loop Duration Mode=::NewEnumerator1, UseLoopDelay=false)
   &SpawnBurst_Instantaneous(Use Spawn Probability=false, UseLoopCountLimit=false) {module=SpawnBurst_Instantaneous003}
   &SpawnBurst_Instantaneous(Spawn Group=1, Use Spawn Probability=false, UseLoopCountLimit=false) {module=SpawnBurst_Instantaneous002}
   &SpawnBurst_Instantaneous(Spawn Group=2, Use Spawn Probability=false, UseLoopCountLimit=false) {module=SpawnBurst_Instantaneous001}
   &SpawnBurst_Instantaneous(Spawn Group=3, Use Spawn Probability=false, UseLoopCountLimit=false)
  ]
  particleSpawnScript [
   &PartitionParticles(ID to Partition:NiagaraInt32=&SpawnGroup(), Number Of Partitions=3)
   &InitializeParticle(
    Color=<rgba 2 1.5312 1.7849 2>
    Lifetime=5
    Ribbon Width=14
    Ribbon Width Mode=::NewEnumerator1
    Sprite Size Mode=::NewEnumerator3
    Uniform Sprite Size=20
    UsePositionOffset=false
    Write Material Random=true
   )
   /SetVariables(Particles.RibbonLinkOrder:NiagaraFloat=&Divide_Float(A=&MakeFloatFromInt(Int=&SpawnGroup()), B=3)) {isSetParameters=true, module=SetVariables_#5}
   &DynamicMaterialParameters(
    Index 0 Param 4=0.5
    Module.Index0Float0Write=false
    Module.Index0Float1Write=false
    Module.Index0Float2Write=false
    Module.Index0Float3Write=true
   )
  ]
  particleUpdateScript [
   &ParticleState(Kill Particles When Lifetime Has Elapsed=false)
   &FN_BoneFlow(
    Angle:NiagaraFloat=&Multiply_Float(A=&Multiply_Float(A=$Particles.PartitionParticles.NormalizedPartition, B=1), B=1)
    BoneID:NiagaraInt32=&ModuloInt(A=$Particles.PartitionParticles.Partition, B=1)
    Enable Position Offsets=true
    Height:NiagaraFloat=&Multiply_Float(A=$Particles.PartitionParticles.NormalizedPartition, B=1)
    Offset End:Vector3f=<-40 -35 0>
    Offset Mid:Vector3f=<-35 0 -20>
    Offset Start:Vector3f=<3 11 0>
    SkeletalMesh:NiagaraDataInterfaceSkeletalMesh=$System.Skeletal Mesh ReaderR
   )
   &ScaleRibbonWidth(Ribbon Width Scale=&FloatFromCurve(
     FloatCurve=DataInterface() {propertyValues {Curve {keys [
         <at -0.013 0 RCIM_Cubic>
         {arriveTangent=0.6934, interpMode=RCIM_Cubic, leaveTangent=0.6934, time=0.0459, value=0.7529}
         {arriveTangent=-0.6133, interpMode=RCIM_Cubic, leaveTangent=-0.6133, time=0.7176, value=0.5066}
         <at 0.9645 0.1895>
        ]}}}
     CurveIndex=$Particles.RibbonLinkOrder
     Scale Curve=&FN_PickFloatDependOnPlatform(TargetPlatformActive=1, TargetPlatformNoActive=1)
    ))
   &ScaleColor(Scale Alpha=&Multiply_Float(A=1.5, B=$System.LODFade), ScaleA=true, ScaleRGB=true)
   /SetVariables(Particles.RibbonWidth02:NiagaraFloat=&Multiply_Float(A=$Particles.RibbonWidth, B=1.5)) {isSetParameters=true, module=SetVariables_#6}
   /SetVariables(Particles.Linear Color02:LinearColor=&Multiply_LinearColorByFloat(Linear Color=$Particles.Color, Float=&Multiply_Float(A=2, B=$System.LODFade))) {isSetParameters=true, module=SetVariables_#7}
  ]
  rendererClasses [&NiagaraRibbonRendererProperties]
  renderers [
   {rendererClass=&NiagaraRibbonRendererProperties}
   {rendererClass=&NiagaraRibbonRendererProperties, rendererIndex=1}
  ]
 }
}
summary {
 systemName=NS_SkyRuled_Skin
 emitters [
  {emitterName=ArmL, rendererClasses [&NiagaraRibbonRendererProperties]}
  {emitterName=ArmR, rendererClasses [&NiagaraRibbonRendererProperties]}
  {emitterName=ArmLSmall, rendererClasses [&NiagaraRibbonRendererProperties]}
  {emitterName=ArmR_Small, rendererClasses [&NiagaraRibbonRendererProperties]}
  {emitterName=Hand, rendererClasses [&NiagaraMeshRendererProperties]}
  {bEnabled=false, emitterName=LinkStableTest, rendererClasses [&NiagaraRibbonRendererProperties]}
 ]
 userVariables [User.bisHandHidden {type=NiagaraBool}]
}
