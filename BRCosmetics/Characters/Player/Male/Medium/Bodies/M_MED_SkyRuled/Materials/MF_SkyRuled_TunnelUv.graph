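# Material function for the SkyRuled player body that produces polar tunnel UVs and a matching tunnel mask.
# Texture coordinates, offset to an adjustable center, feed VectorToRadialValue, and tunnel radius divided by center distance gives the depth coordinate.
# A weighted U/V distortion is added, then a static switch either pans the result or adds the TilingU/V parameters as an offset.
# The mask is a saturated smoothstep of the depth value with adjustable softness.
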MaterialFunction /BRCosmetics/Characters/Player/Male/Medium/Bodies/M_MED_SkyRuled/Materials/MF_SkyRuled_TunnelUv
graph {nodes [
  n1 = TextureCoordinate()
  n2 = Constant() {r=0.5}
  n3 = ScalarParameter() {group=TunnelUV, parameterName=Tunnel CenterX}
  n4 = ScalarParameter() {group=TunnelUV, parameterName=Tunnel CenterY}
  MaterialExpressionMaterialFunctionCall_0 = MaterialFunctionCall(Vector or UVs=Add(A=Subtract(A=@n1, B=@n2) {constB=0.5}, B=AppendVector(A=@n3, B=@n4))) {materialFunction=&VectorToRadialValue}
  n5 = Convert(None=@MaterialExpressionMaterialFunctionCall_0.Radial Coordinates) {
   nodeName=Break Float2
   convertInputs [{
     type=Vector2
     expressionInput {
      expression=&self:MaterialExpressionMaterialFunctionCall_0
      inputName=None
      mask=0
      maskA=0
      maskB=0
      maskG=0
      maskR=0
      outputIndex=0
     }
    }]
   convertOutputs [{type=Scalar}, {type=Scalar}]
  }
  n6 = Saturate(None=Subtract(A=@n5, B=Constant()))
  n7 = Saturate(None=Divide(
    A=ScalarParameter() {defaultValue=0.082, group=TunnelUV, parameterName=Tunnel Radius}
    B=NamedRerouteDeclaration(None=@n6) {name=CenterDistance, nodeColor=<rgba 1 0.7969 0 1>}
   ))
  n8 = NamedRerouteUsage()
  n9 = ScalarParameter() {group=TunnelUV, parameterName=Distorted U}
  n10 = ScalarParameter() {group=TunnelUV, parameterName=Distorted V}
  n11 = Add(A=AppendVector(A=Multiply(A=@n5), B=@n7), B=Multiply(A=OneMinus(None=@n8), B=AppendVector(A=@n9, B=@n10)))
  n12 = AppendVector(
   A=ScalarParameter() {defaultValue=0.1, group=TunnelUV, parameterName=PannerSpeedX}
   B=ScalarParameter() {defaultValue=-0.1, group=TunnelUV, parameterName=PannerSpeedY}
  )
  n13 = AppendVector(A=ScalarParameter() {group=TunnelUV, parameterName=TilingU}, B=ScalarParameter() {group=TunnelUV, parameterName=TilingV})
  FunctionOutput(None=StaticSwitchParameter(True=Panner(Coordinate=@n11, Speed=@n12) {speedX=0.1, speedY=-0.1}, False=Add(A=@n11, B=@n13)) {defaultValue=true, group=TunnelUV, parameterName=Panner Animation}) {outputName=TunnelUV}
  n14 = SmoothStep(Min=ScalarParameter() {defaultValue=0.99, parameterName=MaskSoftness}, Value=@n7)
  FunctionOutput(None=NamedRerouteDeclaration(None=Saturate(None=@n14)) {name=TunnelMask, nodeColor=<rgba 1 0 0.9609 1>}) {outputName=TunnelMask, sortPriority=2}
 ]}
