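# MF_SkyRuled_PolarUV is a material function for the SkyRuled player body cosmetic that converts UVs into polar coordinates.
# It offsets input UVs to a center, then pairs scaled distance with the arctangent angle divided by scaled Pi.
# An optional static-switched twist adds a sphere-masked offset, then sums that result with a diagonally panned copy of itself.
# The twist mask is also exposed as a debug output.
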MaterialFunction /BRCosmetics/Characters/Player/Male/Medium/Bodies/M_MED_SkyRuled/Materials/MF_SkyRuled_PolarUV
graph {nodes [
  n1 = SphereMask(
   A=NamedRerouteUsage()
   B=Constant()
   Radius=FunctionInput(Preview=Constant() {r=0.2033}) {bUsePreviewValueAsDefault=true, inputName=PolarUV_TwistRadius, inputType=FunctionInput_Scalar, sortPriority=4}
  ) {attenuationRadius=0.75, hardnessPercent=0}
  n2 = AppendVector(A=Constant(), B=Constant() {r=-0.55})
  n3 = FunctionInput() {bUsePreviewValueAsDefault=true, inputName=PolarUV_LenghtScale, inputType=FunctionInput_Scalar, sortPriority=2, previewValue=<1 0 0 0>}
  n4 = Constant() {r=-0.5}
  n5 = Constant() {r=-0.5}
  n6 = FunctionInput(Preview=TextureCoordinate()) {bUsePreviewValueAsDefault=true, inputName=UV_Input, inputType=FunctionInput_Vector2, previewValue=<1 1 0 1>}
  n7 = Add(
   A=AppendVector(
    A=FunctionInput(Preview=@n4) {bUsePreviewValueAsDefault=true, inputName=PolarUV_CenterOffsetR, inputType=FunctionInput_Scalar, sortPriority=1}
    B=FunctionInput(Preview=@n5) {bUsePreviewValueAsDefault=true, inputName=PolarUV_CenterOffsetG, inputType=FunctionInput_Scalar, sortPriority=1}
   )
   B=@n6
  )
  n8 = ComponentMask(None=@n7) {r=true}
  n9 = Multiply(A=@n3, B=@n8)
  n10 = Multiply(A=@n9, B=@n9)
  n11 = ComponentMask(None=@n7) {g=true}
  n12 = Multiply(A=@n3, B=@n11)
  n13 = Multiply(A=@n12, B=@n12)
  n14 = FunctionInput() {
   bUsePreviewValueAsDefault=true
   desc=Radial Gradient
   inputName=PolarUV_RadialScale
   inputType=FunctionInput_Scalar
   sortPriority=2
   previewValue=<1 0 0 0>
  }
  n15 = AppendVector(
   A=SquareRoot(None=Add(A=@n10, B=@n13))
   B=Divide(
    A=Arctangent2Fast(Y=@n11, X=@n8)
    B=MaterialFunctionCall(Multiplier=@n14) {desc=Radial Gradient, materialFunction=&Pi}.Result
   )
  )
  n16 = Add(
   A=Multiply(
    A=@n1
    B=FunctionInput(Preview=@n2) {bUsePreviewValueAsDefault=true, inputName=PolarUV_TwistOffset, inputType=FunctionInput_Vector2, sortPriority=4}
   )
   B=@n15
  )
  n17 = Multiply(
   A=Constant() {r=1}
   B=FunctionInput(Preview=Constant() {r=0.05}) {bUsePreviewValueAsDefault=true, inputName=PolarUVTwistSpeed, inputType=FunctionInput_Scalar, sortPriority=4}
  ) {constB=0.05}
  n18 = Panner(Coordinate=@n16, Speed=AppendVector(A=Multiply(A=@n17) {constB=-1}, B=@n17))
  n19 = StaticBool() {value=true}
  FunctionOutput(None=StaticSwitch(
    True=Add(A=@n16, B=@n18)
    False=@n15
    Value=FunctionInput(Preview=@n19) {bUsePreviewValueAsDefault=true, inputName=PolarUV_UseTwist, inputType=FunctionInput_StaticBool, sortPriority=3}
   ))
  FunctionOutput(None=@n1) {outputName=DebugTwistMask}
  NamedRerouteDeclaration(None=@n7) {name=InputUV, nodeColor=<rgba 0.1484 1 0 1>}
 ]}
