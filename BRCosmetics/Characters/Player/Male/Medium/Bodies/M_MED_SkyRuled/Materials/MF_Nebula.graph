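# Material function for the SkyRuled character body that builds an animated, swirling nebula with a glowing core and stars, outputting color and two masks.
# A Custom-node homography warps UVs by four corner parameters, then rotation, noise distortion and twisted polar mapping shape the swirl.
# Packed mask textures, bump-offset nebula layers and sphere masks drive color gradients and core blend.
# Three star layers rotate at different speeds, with panned noise varying brightness.
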MaterialFunction /BRCosmetics/Characters/Player/Male/Medium/Bodies/M_MED_SkyRuled/Materials/MF_Nebula
graph {nodes [
  n1 = TextureCoordinate()
  n2 = TextureSample!(UVs=NamedRerouteUsage(), Tex, "DDX(UVs)"=DDX(Value=@n1), "DDY(UVs)"=DDY(Value=@n1), Apply View MipBias) {
   mipValueMode=TMVM_Derivative
   samplerSource=SSM_Wrap_WorldGroupSettings
   samplerType=SAMPLERTYPE_Masks
   texture=/BRCosmetics/Characters/Player/Male/Medium/Bodies/M_MED_SkyRuled/Textures/T_StarPacked2_RGB
  }
  n3 = Power(Base=@n2.R, Exp=ScalarParameter() {defaultValue=1, group=MF_Nebula, parameterName=Nebula_Power}) {constExponent=1.2}
  n4 = TextureSample!(UVs=NamedRerouteUsage(), Tex, Apply View MipBias) {
   samplerSource=SSM_Wrap_WorldGroupSettings
   samplerType=SAMPLERTYPE_Masks
   texture=&T_StarPacked2_RGB
  }
  n5 = Multiply(A=@n4.B, B=SphereMask(A=NamedRerouteUsage(), B=Constant() {r=0.5}, Radius=Constant() {r=0.142}, Hardness=Constant() {r=0.278}))
  n6 = NamedRerouteUsage()
  n7 = Constant() {r=0.5}
  n8 = ScalarParameter() {defaultValue=0.1725, group=MF_Nebula, parameterName=NebulaCoreRadius}
  n9 = ScalarParameter() {defaultValue=0.06, group=MF_Nebula, parameterName=NebulaCoreHardness}
  n10 = Multiply(A=Power(Base=@n5), B=Power(Base=SphereMask(A=@n6, B=@n7, Radius=@n8, Hardness=@n9)))
  n11 = SphereMask(
   A=NamedRerouteUsage()
   B=Constant() {r=0.5}
   Radius=ScalarParameter() {group=MF_Nebula, parameterName=NebulaCoreRadius2}
   Hardness=ScalarParameter() {group=MF_Nebula, parameterName=NebulaCoreHardness2}
  )
  n12 = Add(A=@n10, B=Multiply(A=Power(Base=@n11)) {constB=0.3})
  n13 = Multiply(
   A=NamedRerouteUsage()
   B=LinearInterpolate(A=ScalarParameter() {defaultValue=0.5, parameterName=Nebula_darkercolor}, Alpha=Add(A=@n3, B=@n12)) {constA=0.3}
  )
  n14 = NamedRerouteUsage()
  n15 = Constant() {r=0.5}
  n16 = Constant() {r=0.35}
  n17 = Constant() {r=0.4347}
  n18 = NamedRerouteUsage()
  n19 = TextureSample(UVs=BumpOffset(Coordinate=@n18, Height=Constant() {r=-0.2798})) {
   samplerSource=SSM_Wrap_WorldGroupSettings
   texture=/Game/Characters/Player/Male/Medium/Bodies/M_Med_Soldier_04/Skins/Galaxy/Textures/T_FN_Nebula_style_blue
  }
  n20 = Constant() {r=-0.5}
  n21 = TextureSample(UVs=Multiply(A=BumpOffset(Coordinate=@n18, Height=@n20), B=Constant() {r=3})) {
   samplerSource=SSM_Wrap_WorldGroupSettings
   texture=&T_FN_Nebula_style_blue
  }
  n22 = TextureSample(UVs=NamedRerouteUsage()) {
   samplerSource=SSM_Wrap_WorldGroupSettings
   samplerType=SAMPLERTYPE_Masks
   texture=&T_StarPacked2_RGB
  }
  n23 = Add(A=@n3, B=Multiply(A=NamedRerouteUsage(), B=NamedRerouteUsage()) {constB=10})
  n24 = Add(A=Multiply(A=SphereMask(A=@n14, B=@n15, Radius=@n16, Hardness=@n17), B=@n19.R), B=Add(A=Multiply(A=@n21.B, B=@n3), B=Add(A=@n22.G, B=@n23)))
  n25 = SphereMask(A=NamedRerouteUsage(), B=Constant() {r=0.5}, Radius=Constant() {r=0.25}, Hardness=Constant() {r=0.25})
  n26 = Add(A=@n5, B=Multiply(A=Add(A=@n24, B=@n25), B=NamedRerouteUsage()))
  n27 = Multiply(
   A=ScalarParameter() {defaultValue=6.9697, group=MF_Nebula, parameterName=NebulaCoreBrightness}
   B=LinearInterpolate(
    A=VectorParameter() {group=MF_Nebula, parameterName=NebulaCoreColorB, defaultValue=<rgba 0.0226 0 2 1>}.RGB
    B=VectorParameter() {group=MF_Nebula, parameterName=NebulaCoreColor, defaultValue=<rgba 0.93 1.0429 2 1>}.RGB
    Alpha=NamedRerouteDeclaration(None=@n10) {name=Nebula_CoreMask, nodeColor=<rgba 1 0 0.5391 1>}
   )
  )
  FunctionOutput(None=LinearInterpolate(A=Multiply(A=@n13, B=@n26), B=Multiply(A=@n27, B=@n12), Alpha=@n12)) {outputName=Nebula_Color}
  MaterialExpressionNamedRerouteUsage_0 = NamedRerouteUsage()
  n28 = VectorParameter() {group=MF_Nebula, parameterName=Nebula_TL_TR, defaultValue=<rgba 0.3 0.9 0.7 0.9>}
  MaterialExpressionComponentMask_2 = ComponentMask(None=@n28.RGBA) {g=true, r=true}
  MaterialExpressionComponentMask_3 = ComponentMask(None=@n28.RGBA) {a=true, b=true}
  n29 = VectorParameter() {group=MF_Nebula, parameterName=Nebula_BL_BR, defaultValue=<rgba 1 0 0 0>}
  MaterialExpressionComponentMask_4 = ComponentMask(None=@n29.RGBA) {g=true, r=true}
  MaterialExpressionComponentMask_5 = ComponentMask(None=@n29.RGBA) {a=true, b=true}
  n30 = Custom!(
   UV=@MaterialExpressionNamedRerouteUsage_0
   TL=@MaterialExpressionComponentMask_2
   TR=@MaterialExpressionComponentMask_3
   BR=@MaterialExpressionComponentMask_4
   BL=@MaterialExpressionComponentMask_5
  ) {
   code="// ----------  ----------\r\nfloat2 p0 = float2(0, 0); // BL\r\nfloat2 p1 = float2(1, 0); // BR\r\nfloat2 p2 = float2(1, 1); // TR\r\nfloat2 p3 = float2(0, 1); // TL\r\n\r\n// ----------  ----------\r\nfloat2 q0 = BL;\r\nfloat2 q1 = BR;\r\nfloat2 q2 = TR;\r\nfloat2 q3 = TL;\r\n\r\n// ---------- Homography ----------\r\nfloat2 d1 = q1 - q2;\r\nfloat2 d2 = q3 - q2;\r\nfloat2 d3 = q0 - q1 + q2 - q3;\r\n\r\nfloat denom = (d1.x * d2.y - d1.y * d2.x);\r\ndenom = sign(denom) * max(abs(denom), 1e-4);\r\n\r\nfloat a13 = (d3.x * d2.y - d3.y * d2.x) / denom;\r\nfloat a23 = (d1.x * d3.y - d1.y * d3.x) / denom;\r\n\r\nfloat2 A = q1 - q0 + a13 * q1;\r\nfloat2 B = q3 - q0 + a23 * q3;\r\nfloat2 C = q0;\r\n\r\n// ----------  ----------\r\nfloat u = UV.x;\r\nfloat v = UV.y;\r\n\r\nfloat w = a13 * u + a23 * v + 1.0;\r\nw = max(w, 1e-4);\r\n\r\nfloat2 warpedUV = (A * u + B * v + C) / w;\r\n\r\n// =====================================================\r\n// Fit \r\n// =====================================================\r\n\r\n// \r\nfloat2 minUV = min(min(q0, q1), min(q2, q3));\r\nfloat2 maxUV = max(max(q0, q1), max(q2, q3));\r\n\r\n// \r\nfloat2 size = max(maxUV - minUV, float2(1e-4, 1e-4));\r\n\r\n//\r\nfloat2 outUV = (warpedUV - minUV) / size;\r\n\r\n// \r\noutUV = saturate(outUV);\r\n\r\nreturn outUV;"
   outputType=CMOT_Float2
   showCode=true
   inputs [
    {
     inputName=UV
     input {expression=&self:MaterialExpressionNamedRerouteUsage_0}
    }
    {
     inputName=TL
     input {expression=&self:MaterialExpressionComponentMask_2}
    }
    {
     inputName=TR
     input {expression=&self:MaterialExpressionComponentMask_3}
    }
    {
     inputName=BR
     input {expression=&self:MaterialExpressionComponentMask_4}
    }
    {
     inputName=BL
     input {expression=&self:MaterialExpressionComponentMask_5}
    }
   ]
  }
  n31 = Time()
  n32 = ScalarParameter() {defaultValue=0.03, group=MF_Nebula, parameterName=Nebula_RotationSpeed}
  n33 = TextureSample() {
   samplerSource=SSM_Wrap_WorldGroupSettings
   samplerType=SAMPLERTYPE_Masks
   texture=&T_StarPacked2_RGB
  }
  n34 = ScalarParameter() {defaultValue=0.025, group=MF_Nebula, parameterName=Nebula_NoiseStrength}
  n35 = SphereMask(A=NamedRerouteUsage(), B=Constant() {r=0.5}, Radius=Constant() {r=0.25}, Hardness=Constant())
  n36 = Add(
   A=MaterialFunctionCall(
    UVs=NamedRerouteDeclaration(None=@n30) {name=WrapUV, nodeColor=<rgba 1 0.6797 0 1>}
    "Rotation Angle (0-1)"=Multiply(A=@n31, B=@n32) {constB=0.02}
   ) {materialFunction=&CustomRotator}.Rotated Values
   B=Multiply(A=@n33.R, B=LinearInterpolate(A=@n34, Alpha=@n35) {constB=0}) {constB=0.05}
  )
  n37 = Constant() {r=1}
  n38 = Multiply(A=@n37, B=Constant() {r=-0.5})
  n39 = MaterialFunctionCall!(
   UV_Input=NamedRerouteDeclaration(None=Multiply(A=@n36, B=@n37) {constB=0.5}) {name=RotatorUV, nodeColor=<rgba 0.875 1 0 1>}
   PolarUV_CenterOffsetG=@n38
   PolarUV_CenterOffsetR=@n38
   PolarUV_LenghtScale=ScalarParameter() {defaultValue=0.552, group=MF_Nebula, parameterName=Nebula_LengthScale}
   PolarUV_RadialScale=Constant() {r=1}
   PolarUV_UseTwist
   PolarUV_TwistOffset=ComponentMask(None=DoubleVectorParameter() {parameterName=Nebula_TwistOffset, defaultValue=<0 0.99 0 1>}.XYZ) {g=true, r=true}
   PolarUVTwistSpeed=ScalarParameter() {group=MF_Nebula, parameterName=Nebula_TwistSpeed}
   PolarUV_TwistRadius=ScalarParameter() {defaultValue=0.16, parameterName=Nebula_TwistRadius}
  ) {materialFunction=&/MF_SkyRuled_PolarUV}
  FunctionOutput(None=NamedRerouteDeclaration(None=@n39.DebugTwistMask) {name=NebulaMask, nodeColor=<rgba 1 0.1875 0 1>}) {outputName=NebulaMask}
  FunctionOutput(None=NamedRerouteDeclaration(None=@n3) {name=NebulaMask_2, nodeColor=<rgba 1 0 0.2812 1>}) {outputName=NebulaMask_2}
  n40 = NamedRerouteUsage()
  n41 = Constant() {r=1.1}
  n42 = LinearInterpolate(
   A=VectorParameter() {group=MF_Nebula, parameterName=Nebula_ColorC, defaultValue=<rgba 0.0029 0 1 1>}.RGB
   B=VectorParameter() {group=MF_Nebula, parameterName=Nebula_ColorA, defaultValue=<rgba 0.8988 0 1.5 1>}.RGB
   Alpha=Saturate(None=Multiply(A=@n40, B=@n41))
  )
  n43 = SphereMask(
   A=NamedRerouteUsage()
   B=Constant() {r=0.5}
   Radius=ScalarParameter() {defaultValue=0.12, group=MF_Nebula, parameterName=Nebula_CenterRadius}
   Hardness=ScalarParameter() {group=MF_Nebula, parameterName=Nebula_CenterHardness}
  )
  n44 = LinearInterpolate(
   A=@n42
   B=VectorParameter() {group=MF_Nebula, parameterName=Nebula_ColorB, defaultValue=<rgba 0 0.6357 5 1>}.RGB
   Alpha=Multiply(A=Power(Base=@n43), B=NamedRerouteUsage())
  )
  LinearInterpolate(A=@n42, B=@n44, Alpha=Power(Base=NamedRerouteUsage()))
  Multiply(A=@n2.R) {constB=1.3}
  NamedRerouteDeclaration(None=@n39.Result) {name=NebulaUV, nodeColor=<rgba 0.4766 0 1 1>}
  n45 = NamedRerouteUsage()
  n46 = TextureSample(UVs=Panner(Coordinate=Multiply(A=@n45) {constB=4}) {speedY=0.1}) {
   samplerSource=SSM_Wrap_WorldGroupSettings
   samplerType=SAMPLERTYPE_Masks
   texture=&T_StarPacked2_RGB
  }
  NamedRerouteDeclaration(None=LinearInterpolate(
    A=ScalarParameter() {defaultValue=0.1, group=MF_Nebula, parameterName=Nebula_StarBrightness_Min}
    B=ScalarParameter() {defaultValue=40, group=MF_Nebula, parameterName=Nebula_StarBrightness_Max}
    Alpha=Power(Base=@n46.R) {constExponent=3}
   ) {constB=10}) {name=Nebula_StarBrightness, nodeColor=<rgba 0 0.8594 1 1>}
  n47 = NamedRerouteUsage()
  n48 = Multiply(A=Time(), B=Constant() {r=0.005}) {constB=0.02}
  n49 = TextureSample(UVs=Multiply(
    A=MaterialFunctionCall(UVs=@n47, "Rotation Angle (0-1)"=@n48) {materialFunction=&CustomRotator}.Rotated Values
    B=ScalarParameter() {defaultValue=4, group=MF_Nebula, parameterName=NebulaStar_mintiling}
   ) {constB=4}) {
   samplerSource=SSM_Wrap_WorldGroupSettings
   samplerType=SAMPLERTYPE_Masks
   texture=/BRCosmetics/Characters/Player/Male/Medium/Bodies/M_MED_SkyRuled/Textures/T_StarPacked_RGB
  }
  n50 = Multiply(A=@n49.R, B=Constant() {r=0.5})
  n51 = NamedRerouteUsage()
  n52 = Multiply(A=Time(), B=Constant() {r=0.006}) {constB=0.02}
  n53 = TextureSample(UVs=Multiply(
    A=MaterialFunctionCall(UVs=@n51, "Rotation Angle (0-1)"=@n52) {materialFunction=&CustomRotator}.Rotated Values
    B=ScalarParameter() {defaultValue=4.5, group=MF_Nebula, parameterName=NebulaStar_midtiling}
   ) {constB=10}) {
   samplerSource=SSM_Wrap_WorldGroupSettings
   samplerType=SAMPLERTYPE_Masks
   texture=&T_StarPacked_RGB
  }
  n54 = NamedRerouteUsage()
  n55 = Multiply(A=Time(), B=Constant() {r=0.003}) {constB=0.02}
  n56 = TextureSample(UVs=Multiply(
    A=MaterialFunctionCall(UVs=@n54, "Rotation Angle (0-1)"=@n55) {materialFunction=&CustomRotator}.Rotated Values
    B=ScalarParameter() {defaultValue=1.5, group=MF_Nebula, parameterName=NebulaStar_maxtiling}
   ) {constB=10}) {
   samplerSource=SSM_Wrap_WorldGroupSettings
   samplerType=SAMPLERTYPE_Masks
   texture=&T_StarPacked_RGB
  }
  n57 = Max(A=Multiply(A=@n53.G) {constB=0.7}, B=Power(Base=@n56.B))
  NamedRerouteDeclaration(None=Power(Base=Add(A=@n50, B=@n57))) {name=Star, nodeColor=<rgba 0.7812 1 0 1>}
  NamedRerouteDeclaration(None=@n44) {name=Nebula_Color, nodeColor=<rgba 0.8984 0 1 1>}
  NamedRerouteDeclaration(None=FunctionInput(Preview=TextureCoordinate()) {bUsePreviewValueAsDefault=true, inputType=FunctionInput_Vector2}) {name=Input_UV, nodeColor=<rgba 0.1484 1 0 1>}
  Power(Base=@n12)
  TextureCoordinate()
  TextureCoordinate()
  TextureCoordinate()
  TextureSample() {texture=/BRCosmetics/Characters/Player/Male/Medium/Bodies/M_MED_NebulaBurden/Texture/T_NebulaBurden_OrbNoise}
 ]}
