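# Material function for the SkyRuled character body that layers an animated galaxy effect over incoming material attributes.
# Two time-panned nebula layers and a star layer share triplanar projection from the NebulaBurden box function, forming emissive with masked Fresnel glow.
# Two sphere masks on pre-skinned position add tinted, Fresnel-modulated hand and arm glow.
# FX texture channels supply masks, and scalar lerps can retain incoming metallic, specular and roughness.
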MaterialFunction /BRCosmetics/Characters/Player/Male/Medium/Bodies/M_MED_SkyRuled/Materials/MF_SkyRuled
graph {nodes [
  FunctionOutput(None=BlendMaterialAttributes(A=NamedRerouteUsage(), B=NamedRerouteUsage(), Alpha=NamedRerouteUsage()))
  n1 = Fresnel(ExponentIn=ScalarParameter() {defaultValue=3, group=Galaxy, parameterName=EmissiveFresnelExponent, sortPriority=90}) {baseReflectFraction=0}
  n2 = ScalarParameter() {defaultValue=1, group=Galaxy, parameterName=EmissiveFresnelIntensity, sortPriority=92}
  n3 = NamedRerouteUsage()
  n4 = Power(Base=NamedRerouteUsage(), Exp=ScalarParameter() {defaultValue=1, group=Galaxy, parameterName=EmissiveMaskExponent, sortPriority=93})
  n5 = Saturate(None=Add(A=ComponentMask(None=@n3) {g=true}, B=@n4))
  n6 = ScalarParameter() {group=Galaxy, parameterName=UseEmissiveMaskOnFresnel, sortPriority=92}
  n7 = NamedRerouteDeclaration(None=Multiply(A=Multiply(A=@n1, B=@n2), B=LinearInterpolate(B=@n5, Alpha=@n6) {constA=1})) {name=FresnelMask, nodeColor=<rgba 0 0.7891 1 1>}
  n8 = VectorParameter() {group=Galaxy, parameterName=EmissiveColor, sortPriority=91, defaultValue=<rgba 0.0885 0.6921 1 1>}
  n9 = Multiply(A=@n4, B=ScalarParameter() {group=Galaxy, parameterName=EmissiveMaskIntensity, sortPriority=92})
  n10 = StaticSwitchParameter(True=NamedRerouteUsage(), False=@n8.RGB) {group=Galaxy, parameterName=UseEmissiveMap, sortPriority=89}
  NamedRerouteDeclaration(None=Add(A=Multiply(A=@n7, B=@n8.RGB), B=Multiply(A=@n9, B=@n10))) {name=Fresnel, nodeColor=<rgba 0.8047 0 1 1>}
  n11 = Multiply(A=NamedRerouteUsage()) {constB=0.2}
  n12 = NamedRerouteUsage()
  NamedRerouteDeclaration(None=Saturate(None=Add(A=@n11, B=@n12))) {name=Galaxies, nodeColor=<rgba 1 0 0.2578 1>}
  n13 = FunctionInput() {inputName=Material, inputType=FunctionInput_MaterialAttributes}
  n14 = GetMaterialAttributes(None=@n13) {attributeGetTypes [
    #1
    #2
    #3
    #4
    #5
   ]}
  n15 = NamedRerouteDeclaration(None=ScalarParameter() {group=Galaxy, parameterName=UseIncomingMatParams}) {name=UseIncomingMatParams, nodeColor=<rgba 0.875 0 1 1>}
  NamedRerouteDeclaration(None=LinearInterpolate(A=ScalarParameter() {defaultValue=0.2, group=Galaxy, parameterName=Galaxy_Roughness, sortPriority=52}, B=@n14.Roughness, Alpha=@n15) {constB=0}) {name=OUTRoughness, nodeColor=<rgba 0 0.4844 1 1>}
  n16 = TextureObjectParameter() {
   group=Galaxy
   parameterName=StarsTexture
   samplerType=SAMPLERTYPE_Masks
   sortPriority=80
   texture=/Game/Characters/Player/Male/Medium/Bodies/M_Med_Soldier_04/Skins/Galaxy/Textures/Celestial_T_Stars
  }
  n17 = VectorParameter() {group=Galaxy, parameterName=Stars_RotationSpeed, sortPriority=84, defaultValue=<rgba 0.06 0.06 0.06 1>}
  n18 = MaterialFunctionCall!(
   Texture=@n16
   Depth=ScalarParameter() {defaultValue=150, group=Galaxy, parameterName=Stars_Depth, sortPriority=81}
   Tile=ScalarParameter() {defaultValue=2, group=Galaxy, parameterName=Stars_Tile, sortPriority=82}
   BlendContrast=ScalarParameter() {defaultValue=1, group=Galaxy, parameterName=Stars_BlendContrast, sortPriority=83}
   SpeedXY=ComponentMask(None=VectorParameter() {group=Galaxy, parameterName=Stars_Speed, sortPriority=84, defaultValue=<rgba 0.002 -0.03 0 1>}.RGB) {g=true, r=true}
   RotationSpeed=@n17.RGB
   ActorOffset=NamedRerouteDeclaration(None=VectorParameter() {group=Galaxy, parameterName=Galaxy_Offset, sortPriority=59, defaultValue=<rgba 0 0.8 0 1>}.RGB) {name=ActorOffset, nodeColor=<rgba 0.9688 1 0 1>}
  ) {materialFunction=/BRCosmetics/Characters/Player/Male/Medium/Bodies/M_MED_NebulaBurden/Material/MF_NebulaBurdenTriplanarBox}
  NamedRerouteDeclaration(None=@n18.UVX) {name=UVX, nodeColor=<rgba 0 0.8828 1 1>}
  NamedRerouteDeclaration(None=@n18.UVY) {name=UVY, nodeColor=<rgba 0 0.8828 1 1>}
  NamedRerouteDeclaration(None=@n18.UVZ) {name=UVZ, nodeColor=<rgba 0 0.8828 1 1>}
  NamedRerouteDeclaration(None=@n18.BlendY) {name=BlendY, nodeColor=<rgba 1 0.6226 0.1707 1>}
  NamedRerouteDeclaration(None=@n18.BlendZ) {name=BlendZ, nodeColor=<rgba 1 0.6226 0.1707 1>}
  n19 = TextureObjectParameter() {
   group=Galaxy
   parameterName=Galaxy_01_Texture
   sortPriority=58
   texture=/Game/Characters/Player/Male/Medium/Bodies/M_Med_Soldier_04/Skins/Galaxy/Textures/T_FN_Nebula_style_blue
  }
  NamedRerouteDeclaration(None=@n19) {name=GalaxyTexture, nodeColor=<rgba 0 1 0.9062 1>}
  NamedRerouteDeclaration(None=Multiply(A=ScalarParameter() {defaultValue=0.4, group=Galaxy, parameterName=Galaxy_01_Tile, sortPriority=62}) {constB=0.01}) {name=GalaxyTile, nodeColor=<rgba 1 0 0.1172 1>}
  n20 = VectorParameter() {group=Galaxy, parameterName=Galaxy_01_Speed, sortPriority=64, defaultValue=<rgba 0.002 -0.0025 0 1>}
  NamedRerouteDeclaration(None=Multiply(A=ComponentMask(None=@n20.RGB) {g=true, r=true}, B=Time())) {name=GalaxySpeed, nodeColor=<rgba 0 0.8125 1 1>}
  NamedRerouteDeclaration(None=LinearInterpolate(A=ScalarParameter() {defaultValue=0.1, group=Galaxy, parameterName=Galaxy_Metallic, sortPriority=50}, B=@n14.Metallic, Alpha=@n15) {constB=0}) {name=OUTMetallic, nodeColor=<rgba 0 0.4844 1 1>}
  n21 = ComponentMask(None=@n18.Result) {g=true}
  n22 = ComponentMask(None=@n18.Result) {b=true}
  NamedRerouteDeclaration(None=Multiply(A=Multiply(A=@n21, B=@n22), B=ScalarParameter() {defaultValue=1, group=Galaxy, parameterName=Stars_Brightness, sortPriority=84})) {name=Stars, nodeColor=<rgba 0 1 0.5078 1>}
  n23 = Multiply(A=NamedRerouteUsage(), B=NamedRerouteUsage())
  n24 = NamedRerouteUsage()
  n25 = Multiply(A=NamedRerouteUsage(), B=NamedRerouteUsage())
  n26 = NamedRerouteUsage()
  n27 = LinearInterpolate(
   A=TextureSample(UVs=Add(A=@n23, B=@n24), Tex=NamedRerouteUsage()) {samplerSource=SSM_Wrap_WorldGroupSettings}.RGB
   B=TextureSample(UVs=Add(A=@n25, B=@n26), Tex=NamedRerouteUsage()) {samplerSource=SSM_Wrap_WorldGroupSettings}.RGB
   Alpha=NamedRerouteUsage()
  )
  n28 = NamedRerouteUsage()
  n29 = NamedRerouteUsage()
  n30 = TextureSample(UVs=Add(A=Multiply(A=@n28, B=@n29), B=NamedRerouteUsage()), Tex=NamedRerouteUsage()) {samplerSource=SSM_Wrap_WorldGroupSettings}
  n31 = NamedRerouteUsage()
  NamedRerouteDeclaration(None=Power(
    Base=LinearInterpolate(A=@n27, B=@n30.RGB, Alpha=@n31)
    Exp=ScalarParameter() {defaultValue=1.75, group=Galaxy, parameterName=Galaxy_01_Exp, sortPriority=66}
   )) {name=Galaxy_01, nodeColor=<rgba 1 0.2578 0 1>}
  NamedRerouteDeclaration(None=LinearInterpolate(A=ScalarParameter() {defaultValue=0.5, group=Galaxy, parameterName=Galaxy_Spec, sortPriority=51}, B=@n14.Specular, Alpha=@n15) {constB=0}) {name=OUTSpec, nodeColor=<rgba 0 0.4844 1 1>}
  NamedRerouteDeclaration(None=@n14.EmissiveColor) {name=IN_Emissive, nodeColor=<rgba 0.4766 1 0 1>}
  n32 = TextureSampleParameter2D() {
   group=Galaxy
   parameterName=MaskTexture
   samplerType=SAMPLERTYPE_Masks
   sortPriority=-20
   texture=/BRCosmetics/Characters/Player/Male/Medium/Bodies/M_MED_SkyRuled/Textures/T_SkyRuled_Body_FX
  }
  n33 = ChannelMaskParameter(None=@n32.RGB) {group=Galaxy, maskChannel=Blue, parameterName=GalaxyMaskChannel, sortPriority=-7, defaultValue=<rgba 0 0 1 0>}
  NamedRerouteDeclaration(None=LinearInterpolate(B=@n33.R, Alpha=ScalarParameter() {defaultValue=1, group=Galaxy, parameterName=GalaxyMaskStrength, sortPriority=-6}) {constA=1}) {name=GalaxyMask, nodeColor=<rgba 0.3125 0 1 1>}
  n34 = Multiply(A=OneMinus(None=NamedRerouteUsage()), B=NamedRerouteUsage())
  n35 = StaticSwitchParameter(True=Multiply(A=NamedRerouteUsage(), B=@n34), False=@n34) {group=Galaxy, parameterName=UseGalaxyMask, sortPriority=-8}
  n36 = NamedRerouteUsage()
  NamedRerouteDeclaration(None=Add(A=NamedRerouteUsage(), B=Add(A=@n35, B=@n36))) {name=OUT_Emissive, nodeColor=<rgba 1 0 0.1875 1>}
  NamedRerouteDeclaration(None=LinearInterpolate(
    A=VectorParameter() {group=Galaxy, parameterName=Galaxy_BaseColor, sortPriority=49, defaultValue=<rgba 0 0 0 1>}.RGB
    B=@n14.EmissiveColor
    Alpha=@n15
   ) {constB=0}) {name=OUTBaseColor, nodeColor=<rgba 1 0 0.6562 1>}
  n37 = FunctionInput() {bUsePreviewValueAsDefault=true, inputName=TechArtMask, sortPriority=1, previewValue=<1 1 1 1>}
  NamedRerouteDeclaration(None=@n37) {name=TechArtMask, nodeColor=<rgba 0 1 0.6953 1>}
  NamedRerouteDeclaration(None=@n21) {name=Noise, nodeColor=<rgba 0 1 0.0156 1>}
  n38 = TextureObjectParameter() {
   group=Galaxy
   parameterName=Galaxy_02_Texture
   sortPriority=58
   texture=/Game/Characters/Player/Male/Medium/Bodies/M_Med_Soldier_04/Skins/Galaxy/Textures/T_FN_Nebula_style2
  }
  NamedRerouteDeclaration(None=@n38) {name=GalaxyTexture_2, nodeColor=<rgba 0 1 0.9062 1>}
  NamedRerouteDeclaration(None=Multiply(A=ScalarParameter() {defaultValue=0.4, group=Galaxy, parameterName=Galaxy_02_Tile, sortPriority=62}) {constB=0.01}) {name=GalaxyTile_1, nodeColor=<rgba 1 0 0.1172 1>}
  n39 = VectorParameter() {group=Galaxy, parameterName=Galaxy_02_Speed, sortPriority=64, defaultValue=<rgba -0.002 -0.0065 0 1>}
  NamedRerouteDeclaration(None=Multiply(A=ComponentMask(None=@n39.RGB) {g=true, r=true}, B=Time())) {name=GalaxySpeed_1, nodeColor=<rgba 0 0.8125 1 1>}
  n40 = SetMaterialAttributes!(
   MaterialAttributes=NamedRerouteDeclaration(None=@n13) {name=IN_Material, nodeColor=<rgba 1 0 0.9609 1>}
   BaseColor=NamedRerouteUsage()
   Metallic=NamedRerouteUsage()
   Specular=NamedRerouteUsage()
   Roughness=NamedRerouteUsage()
   EmissiveColor=NamedRerouteUsage()
  ) {attributeSetTypes [
    #5
    #2
    #3
    #1
    #4
   ]}
  NamedRerouteDeclaration(None=@n40) {name=MaterialOUT, nodeColor=<rgba 0 1 0.625 1>}
  n41 = FunctionInput(Preview=Constant3Vector()) {inputName=PreskinPosition}
  n42 = Constant()
  n43 = VectorParameter() {group=Galaxy, parameterName=Hand Position Offset2, defaultValue=<rgba -62.8927 0 88.1202 1>}
  n44 = Constant()
  n45 = VectorParameter() {group=Galaxy, parameterName=Hand Position Offset, defaultValue=<rgba 62.8927 0 88.1202 1>}
  n46 = Max(
   A=SphereMask(
    A=VertexInterpolator(VS=@n41).PS
    B=Add(A=@n42, B=@n43.RGB)
    Radius=ScalarParameter() {defaultValue=52.9138, group=Galaxy, parameterName=HandRadius}
    Hardness=ScalarParameter() {defaultValue=0.466, group=Galaxy, parameterName=HandHardness}
   )
   B=SphereMask(
    A=VertexInterpolator(VS=@n41).PS
    B=Add(A=@n44, B=@n45.RGB)
    Radius=ScalarParameter() {defaultValue=52.91, group=Galaxy, parameterName=HandRadius}
    Hardness=ScalarParameter() {defaultValue=0.466, group=Galaxy, parameterName=HandHardness}
   )
  )
  n47 = ScalarParameter() {defaultValue=3.3804, group=Galaxy, parameterName=Hand Mask Power}
  n48 = Multiply(
   A=LinearInterpolate(A=ScalarParameter() {defaultValue=0.054, group=Galaxy, parameterName=Hand brightness}, Alpha=NamedRerouteUsage()) {constA=0.2}
   B=NamedRerouteDeclaration(None=Power(Base=@n46, Exp=@n47)) {name=HandMask, nodeColor=<rgba 1 0 0.0469 1>}
  )
  n49 = VectorParameter() {group=Galaxy, parameterName=HandEmissiveColor, sortPriority=91, defaultValue=<rgba 0.0885 0.6921 1 1>}
  n50 = VectorParameter() {group=Galaxy, parameterName=HandEmissiveColor2, sortPriority=91, defaultValue=<rgba 0.2773 0.113 1 1>}
  n51 = NamedRerouteUsage()
  n52 = Multiply(
   A=LinearInterpolate(
    A=VectorParameter() {group=Galaxy, parameterName=ArmEmissiveColor, defaultValue=<rgba 0.0597 0 1 1>}.RGB
    B=LinearInterpolate(A=@n49.RGB, B=@n50.RGB, Alpha=@n51)
    Alpha=NamedRerouteUsage()
   )
   B=ScalarParameter() {defaultValue=7.6, group=Galaxy, parameterName=EmissiveColor_mul}
  ) {constB=3}
  n53 = ScalarParameter() {defaultValue=0.2739, group=Galaxy, parameterName=EmissiveHandFresnelExponent, sortPriority=90}
  n54 = LinearInterpolate(
   A=ScalarParameter() {defaultValue=0.5, group=Galaxy, parameterName=HandFresnelBrightness}
   Alpha=Power(Base=Fresnel(ExponentIn=@n53) {baseReflectFraction=0})
  ) {constA=0.5}
  NamedRerouteDeclaration(None=Multiply(A=Multiply(A=@n48, B=@n52), B=Saturate(None=@n54))) {name=Hand, nodeColor=<rgba 0.7109 0 1 1>}
  n55 = ChannelMaskParameter(None=@n32.RGB) {group=Galaxy, maskChannel=Blue, parameterName=AlphaChannel, sortPriority=-10, defaultValue=<rgba 0 0 1 0>}
  NamedRerouteDeclaration(None=Saturate(None=Multiply(A=@n55.R) {constB=5})) {name=Alpha, nodeColor=<rgba 0.7109 1 0 1>}
  n56 = ChannelMaskParameter(None=NamedRerouteUsage()) {group=Galaxy, maskChannel=Blue, parameterName=StarMaskChannel, sortPriority=84, defaultValue=<rgba 0 0 1 0>}
  NamedRerouteDeclaration(None=Multiply(A=@n56.R, B=ScalarParameter() {defaultValue=3, group=Galaxy, parameterName=GalaxyStarMaskStrength, sortPriority=85}) {constB=2}) {name=StarMask_1, nodeColor=<rgba 0.9922 1 0 1>}
  n57 = Add(A=Multiply(A=NamedRerouteUsage(), B=NamedRerouteUsage()), B=NamedRerouteUsage())
  n58 = NamedRerouteUsage()
  n59 = Add(A=Multiply(A=NamedRerouteUsage(), B=NamedRerouteUsage()), B=NamedRerouteUsage())
  n60 = NamedRerouteUsage()
  n61 = Multiply(A=NamedRerouteUsage(), B=NamedRerouteUsage())
  n62 = NamedRerouteUsage()
  n63 = LinearInterpolate(
   A=LinearInterpolate(
    A=TextureSample(UVs=@n57, Tex=@n58) {samplerSource=SSM_Wrap_WorldGroupSettings}.RGB
    B=TextureSample(UVs=@n59, Tex=@n60) {samplerSource=SSM_Wrap_WorldGroupSettings}.RGB
    Alpha=NamedRerouteUsage()
   )
   B=TextureSample(UVs=Add(A=@n61, B=@n62), Tex=NamedRerouteUsage()) {samplerSource=SSM_Wrap_WorldGroupSettings}.RGB
   Alpha=NamedRerouteUsage()
  )
  n64 = ScalarParameter() {defaultValue=1.75, group=Galaxy, parameterName=Galaxy_02Exp, sortPriority=66}
  NamedRerouteDeclaration(None=Multiply(A=Power(Base=@n63, Exp=@n64)) {constB=0.5}) {name=Galaxy_02, nodeColor=<rgba 0.0781 0 1 1>}
  n65 = LinearInterpolate(A=NamedRerouteUsage(), Alpha=NamedRerouteUsage()) {constB=0}
  NamedRerouteDeclaration(None=Add(A=Saturate(None=@n65), B=NamedRerouteUsage())) {name=GalaxiesAndStars, nodeColor=<rgba 0 1 0.9063 1>}
  n66 = ChannelMaskParameter(None=@n32.RGB) {group=Galaxy, maskChannel=Blue, parameterName=EmissiveMaskChannel, sortPriority=-9, defaultValue=<rgba 0 0 1 0>}
  NamedRerouteDeclaration(None=@n66.R) {name=EmissiveMask, nodeColor=<rgba 0 0.4375 1 1>}
 ]}
