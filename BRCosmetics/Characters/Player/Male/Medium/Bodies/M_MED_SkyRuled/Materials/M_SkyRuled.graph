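# Character material for the SkyRuled male medium body, adding an emissive nebula effect around an orb.
# MF_SkyRuled runs between shared character PreFX and PostFX functions; MF_Nebula's mask dims its emissive, then nebula color is added through the tech-art mask's blue channel.
# Named reroutes build a camera-facing virtual plane for projected orb UVs.
# Orb center blends between a fixed mesh offset and parameter-driven location and axes.
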Material /BRCosmetics/Characters/Player/Male/Medium/Bodies/M_MED_SkyRuled/Materials/M_SkyRuled
graph {
 properties {
  bUseMaterialAttributes=true
  bUsedWithInstancedStaticMeshes=true
  bUsedWithMeshDeformer=true
  bUsedWithMorphTargets=true
  bUsedWithNanite=true
  bUsedWithSkeletalMesh=true
  pixelDepthOffsetMode=PDOM_Legacy
  shadingModel=MSM_FromMaterialExpression
  usedShadingModels=Default Lit | Subsurface | Clear Coat | Subsurface Profile | Cloth
 }
 nodes [
  n1 = MaterialFunctionCall() {materialFunction=&MF_CharacterParent_2023_PreFX}
  n2 = GetMaterialAttributes(None=MaterialFunctionCall!(PreskinPosition=@n1.Pre skinned Local Position, Material=@n1.Pre skinned Local Position, TechArtMask=@n1.Pre skinned Local Position) {materialFunction=&/MF_SkyRuled}.Result) {attributeGetTypes [#1]}
  n3 = AppendVector(
   A=ScalarParameter() {defaultValue=0.28, group=CameraAlignedVirtualPlane, parameterName=UVScaleX, sortPriority=0}
   B=ScalarParameter() {defaultValue=0.25, group=CameraAlignedVirtualPlane, parameterName=UVScaleY, sortPriority=1}
  )
  n4 = NamedRerouteUsage()
  n5 = VectorParameter() {parameterName=UVOffset, sortPriority=2, defaultValue=<rgba 0.358 0.144 0 1>}
  n6 = MaterialFunctionCall!(In=Add(A=Multiply(A=@n3, B=@n4), B=ComponentMask(None=@n5.RGB) {g=true, r=true})) {materialFunction=&/MF_Nebula}
  n7 = Multiply(A=@n6.NebulaMask) {constB=1.5}
  n8 = Multiply(A=@n2.Emissive Color, B=LinearInterpolate(Alpha=Saturate(None=@n7)) {constA=1, constB=0})
  n9 = Multiply(A=@n6.Nebula_Color, B=ComponentMask(None=@n1.Tech Art Mask) {b=true})
  MaterialFunctionCall(
   Material=SetMaterialAttributes(MaterialAttributes=@n2.MaterialAttributes, Emissive Color=Add(A=@n8, B=@n9)) {attributeSetTypes [#1]}
   M=@n1.M
   Pre Skinned Local Position=@n1.M
   Distance Blend=@n1.M
   Fresnel=@n1.M
  ) {materialFunction=&MF_CharacterParent_2023_PostFX}
  n10 = NamedRerouteUsage()
  n11 = CameraPositionWS()
  n12 = NamedRerouteDeclaration(None=Subtract(A=NamedRerouteUsage(), B=NamedRerouteDeclaration(None=@n11) {name=CamPos, nodeColor=<rgba 1 0.9844 0 1>})) {name=CamToCenter, nodeColor=<rgba 1 0 0.75 1>}
  n13 = NamedRerouteUsage()
  n14 = CameraVectorWS()
  n15 = Multiply(A=Divide(A=DotProduct(A=@n12, B=@n13), B=DotProduct(A=@n13, B=@n14)), B=@n14)
  NamedRerouteDeclaration(None=Subtract(A=Add(A=@n10, B=@n15), B=NamedRerouteUsage())) {name=CenteredWSP, nodeColor=<rgba 0.4297 1 0 1>}
  n16 = NamedRerouteDeclaration(None=CameraVectorWS()) {name=PlaneNormal, nodeColor=<rgba 1 0 0.1172 1>}
  n17 = Transform(None=Constant3Vector() {constant=<rgba 0 1 0 0>}) {transformSourceType=TRANSFORMSOURCE_View}
  NamedRerouteDeclaration(None=Normalize(VectorInput=CrossProduct(A=@n16, B=@n17))) {name=PlaneXAxis, nodeColor=<rgba 0.7812 0 1 1>}
  n18 = NamedRerouteUsage()
  n19 = CrossProduct(A=NamedRerouteUsage(), B=NamedRerouteUsage())
  n20 = AppendVector(A=DotProduct(A=NamedRerouteUsage(), B=@n18), B=DotProduct(A=@n18, B=Normalize(VectorInput=@n19)))
  n21 = ScalarParameter() {defaultValue=28, group=Orb, parameterName=OrbUVDivisor}
  NamedRerouteDeclaration(None=Add(A=Divide(A=@n20, B=@n21)) {constB=0.5}) {name=OrbUVs, nodeColor=<rgba 1 0.7734 0 1>}
  n22 = TransformPosition(None=Constant3Vector()) {transformType=TRANSFORMPOSSOURCE_World}
  n23 = VectorParameter() {group=Orb, parameterName=OrbOffset, defaultValue=<rgba 0 -5 128 1>}
  n24 = VectorParameter() {group=Orb, parameterName=OrbLocation, defaultValue=<rgba 0 0 0 1>}
  n25 = VectorParameter() {group=Orb, parameterName=OrbUpAxis, defaultValue=<rgba 0 0 1 1>}
  n26 = Add(
   A=Multiply(
    A=NamedRerouteDeclaration(None=@n25.RGB) {name=OrbUpVector, nodeColor=<rgba 0 0.8125 1 1>}
    B=ScalarParameter() {defaultValue=-3, group=Orb, parameterName=OrbOffsetUpRuntime}
   )
   B=Multiply(
    A=ScalarParameter() {defaultValue=-2, group=Orb, parameterName=OrbOffsetForwardRuntime}
    B=VectorParameter() {group=Orb, parameterName=OrbForwardAxis, defaultValue=<rgba 0 1 0 1>}.RGB
   )
  )
  NamedRerouteDeclaration(None=LinearInterpolate(A=Add(A=@n22, B=@n23.RGB), B=Add(A=@n24.RGB, B=@n26), Alpha=ScalarParameter() {group=Orb, parameterName=UseCCPMOrbLocation})) {name=OrbCenter, nodeColor=<rgba 1 0.6797 0 1>}
 ]
}
