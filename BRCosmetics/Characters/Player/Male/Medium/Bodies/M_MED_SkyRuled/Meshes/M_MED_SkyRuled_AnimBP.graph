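# Animation Blueprint for the SkyRuled male body, adding bag and scarf secondary motion.
# The parent's copied pose runs through a Control Rig, weighted scarf constraints, bone offsets and left-scarf AnimDynamics into a cached pose.
# Without back bling, right-scarf AnimDynamics with spherical limits follows; with it, static offsets and a constraint reposition the scarf instead.
# The event graph detects back bling and computes skydive gravity, thigh and head values the AnimGraph never reads.
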AnimBlueprint /BRCosmetics/Characters/Player/Male/Medium/Bodies/M_MED_SkyRuled/Meshes/M_MED_SkyRuled_AnimBP
target_skeleton=&/M_MED_SkyRuled_Skeleton
graphs [
 AnimGraph {nodes [
   n1 = Anim Dynamics - Bone: dyn_scarf_01_R!(ComponentPose=Local To Component(LocalPose=Use cached pose 'start cache'()), Alpha=1, GravityOverride:Vector="0,0,-980") -> (Pose) {
    angularBiasOverride=1
    angularDampingOverride=1
    angularSpringConstant=1000
    bAngularSpring=true
    bOverrideAngularBias=true
    bOverrideAngularDamping=true
    bUseGravityOverride=true
    bUseSphericalLimits=true
    simulationSpace=RootRelative
    alphaBoolBlend {blendOption=Linear}
    boundBone {boneName=dyn_scarf_01_R}
    componentAppliedLinearAccClamp=<1000 1000 1000>
    componentLinearAccScale=<0.5 0.5 0.5>
    componentLinearVelScale=<0.5 0.5 0.5>
    physicsBodyDefinitions [{
      collisionType=CustomSphere
      sphereCollisionRadius=1
      boundBone {boneName=dyn_scarf_01_R}
      boxExtents=<15 15 15>
      constraintSetup {angularLimitsMax=<0 0 0>, angularLimitsMin=<0 -35 0>, angularTarget=<1 0 0>}
      localJointOffset=<2 0 0>
     }]
    simSpaceSettings {maxAngularAcceleration=50, maxAngularVelocity=8, simSpaceAngularAlpha=0.5}
    sphericalLimits [
     {limitRadius=11, drivingBone {boneName=clavicle_r}, sphereLocalOffset=<0 0 0>}
     {limitRadius=7.5, drivingBone {boneName=neck_01}, sphereLocalOffset=<0 0 0>}
    ]
   }
   n2 = Anim Dynamics - Bone: dyn_scarf_02_R!(ComponentPose=@n1, Alpha=1, GravityOverride:Vector="0,0,-980") -> (Pose) {
    angularBiasOverride=1
    angularDampingOverride=1
    angularSpringConstant=2000
    bAngularSpring=true
    bOverrideAngularBias=true
    bOverrideAngularDamping=true
    bUseGravityOverride=true
    bUseSphericalLimits=true
    simulationSpace=RootRelative
    alphaBoolBlend {blendOption=Linear}
    boundBone {boneName=dyn_scarf_02_R}
    componentAppliedLinearAccClamp=<1000 1000 1000>
    componentLinearAccScale=<0.5 0.5 0.5>
    componentLinearVelScale=<0.5 0.5 0.5>
    physicsBodyDefinitions [{
      collisionType=CustomSphere
      sphereCollisionRadius=1
      boundBone {boneName=dyn_scarf_02_R}
      boxExtents=<10 10 10>
      constraintSetup {angularLimitsMax=<0 15 0>, angularLimitsMin=<0 -25 0>, angularTarget=<1 0 0>}
      localJointOffset=<2 0 0>
     }]
    simSpaceSettings {maxAngularAcceleration=50, maxAngularVelocity=8, simSpaceAngularAlpha=0.5}
    sphericalLimits [{limitRadius=11.5, drivingBone {boneName=clavicle_l}, sphereLocalOffset=<5 0 0>}]
   }
   Output Pose(Result=Blend Poses by bool(
     BlendPose_0=Use cached pose 'backpack cache'()
     BlendPose_1=Component To Local(ComponentPose=@n2)
     bActiveValue=Get isWearingBackpack() -> (isWearingBackpack:bool)
    ) {blendType=Linear}) {layerGroup=None}
   n3 = Control Rig(Source=Copy Pose From Mesh() {bUseAttachedParent=true}) {
    alphaBoolBlend {blendOption=Linear}
    controlRigAssetReference {blueprintRigClass=&/M_MED_SkyRuled_ControlRig.M_MED_SkyRuled_ControlRig_C}
   }
   n4 = "Constraint - dyn_scarf_CST_L (PPPPPPPPPPPP)"!(
    ComponentPose=Local To Component(LocalPose=@n3)
    Alpha=1
    ConstraintWeights_0=0
    ConstraintWeights_1:float=0
    ConstraintWeights_2:float=0.0031
    ConstraintWeights_3:float=0.118
    ConstraintWeights_4:float=0.2571
    ConstraintWeights_5:float=0.0567
    ConstraintWeights_6:float=0.4146
    ConstraintWeights_7:float=0
    ConstraintWeights_8:float=0.0034
    ConstraintWeights_9:float=0.0019
    ConstraintWeights_10:float=0.1297
    ConstraintWeights_11:float=0.0154
   ) -> (Pose) {
    boneToModify {boneName=dyn_scarf_CST_L}
    constraintSetup [
     {targetBone {boneName=spine_02}}
     {targetBone {boneName=spine_03}}
     {targetBone {boneName=spine_04}}
     {targetBone {boneName=spine_05}}
     {targetBone {boneName=neck_01}}
     {targetBone {boneName=neck_02}}
     {targetBone {boneName=clavicle_l}}
     {targetBone {boneName=upperarm_l}}
     {targetBone {boneName=upperarm_twist_01_l}}
     {targetBone {boneName=clavicle_r}}
     {targetBone {boneName=dyn_scarf_01_L}}
     {targetBone {boneName=dyn_scarf_02_L}}
    ]
    constraintWeights [1,1,1,1,1,1,1,1,1,1,1,1]
   }
   n5 = "Constraint - dyn_scarf_CST_R (PPPPPPPPPPPPPP)"!(
    ComponentPose=@n4
    Alpha=1
    ConstraintWeights_0=0
    ConstraintWeights_1:float=0
    ConstraintWeights_2:float=0.0007
    ConstraintWeights_3:float=0.5551
    ConstraintWeights_4:float=0.1616
    ConstraintWeights_5:float=0.0197
    ConstraintWeights_6:float=0
    ConstraintWeights_7:float=0.002
    ConstraintWeights_8:float=0.1944
    ConstraintWeights_9:float=0
    ConstraintWeights_10:float=0.0002
    ConstraintWeights_11:float=0
    ConstraintWeights_12:float=0.0565
    ConstraintWeights_13:float=0.0096
   ) -> (Pose) {
    boneToModify {boneName=dyn_scarf_CST_R}
    constraintSetup [
     {targetBone {boneName=spine_02}}
     {targetBone {boneName=spine_03}}
     {targetBone {boneName=spine_04}}
     {targetBone {boneName=spine_05}}
     {targetBone {boneName=neck_01}}
     {targetBone {boneName=neck_02}}
     {targetBone {boneName=head}}
     {targetBone {boneName=clavicle_l}}
     {targetBone {boneName=clavicle_r}}
     {targetBone {boneName=upperarm_r}}
     {targetBone {boneName=upperarm_twist_01_r}}
     {targetBone {boneName=upperarm_twist_02_r}}
     {targetBone {boneName=dyn_scarf_01_R}}
     {targetBone {boneName=dyn_scarf_02_R}}
    ]
    constraintWeights [1,1,1,1,1,1,1,1,1,1,1,1,1,1]
   }
   n6 = "Transform (Modify) Bone - Bone: bag_L"!(ComponentPose=@n5, Alpha=1, Rotation="20,0,-20") -> (Pose) {
    rotationMode=BMM_Additive
    rotationSpace=BCS_BoneSpace
    alphaBoolBlend {blendOption=Linear}
    boneToModify {boneName=bag_L}
    rotation {pitch=20, roll=-20}
   }
   n7 = "Transform (Modify) Bone - Bone: dyn_scarf_MSTR_R"!(ComponentPose=@n6, Alpha=1, Translation="-1.371,-0.725,4.047", Rotation="0,0,0") -> (Pose) {
    rotationMode=BMM_Additive
    rotationSpace=BCS_BoneSpace
    translationMode=BMM_Additive
    translationSpace=BCS_BoneSpace
    alphaBoolBlend {blendOption=Linear}
    boneToModify {boneName=dyn_scarf_MSTR_R}
    translation=<-1.371 -0.725 4.047>
   }
   n8 = Anim Dynamics - Bone: dyn_scarf_01_L!(ComponentPose=@n7, Alpha=1, GravityOverride:Vector="0,0,-980") -> (Pose) {
    angularBiasOverride=1
    angularDampingOverride=1
    angularSpringConstant=1000
    bAngularSpring=true
    bOverrideAngularBias=true
    bOverrideAngularDamping=true
    bUseGravityOverride=true
    simulationSpace=RootRelative
    alphaBoolBlend {blendOption=Linear}
    boundBone {boneName=dyn_scarf_01_L}
    componentAppliedLinearAccClamp=<1000 1000 1000>
    componentLinearAccScale=<0.5 0.5 0.5>
    componentLinearVelScale=<0.5 0.5 0.5>
    physicsBodyDefinitions [{
      boundBone {boneName=dyn_scarf_01_L}
      boxExtents=<10 10 10>
      constraintSetup {angularLimitsMax=<0 25 0>, angularLimitsMin=<0 -45 0>, angularTarget=<1 0 0>}
      localJointOffset=<0 0 0>
     }]
    simSpaceSettings {maxAngularAcceleration=50, maxAngularVelocity=8, simSpaceAngularAlpha=0.5}
   }
   n9 = Anim Dynamics - Bone: dyn_scarf_02_L!(ComponentPose=@n8, Alpha=1, GravityOverride:Vector="0,0,-980") -> (Pose) {
    angularBiasOverride=1
    angularDampingOverride=1
    angularSpringConstant=300
    bAngularSpring=true
    bOverrideAngularBias=true
    bOverrideAngularDamping=true
    bUseGravityOverride=true
    simulationSpace=RootRelative
    alphaBoolBlend {blendOption=Linear}
    boundBone {boneName=dyn_scarf_02_L}
    componentAppliedLinearAccClamp=<3000 3000 3000>
    componentLinearAccScale=<0.5 0.5 0.5>
    componentLinearVelScale=<0.5 0.5 0.5>
    physicsBodyDefinitions [{
      boundBone {boneName=dyn_scarf_02_L}
      boxExtents=<5 5 5>
      constraintSetup {angularLimitsMax=<0 10 0>, angularLimitsMin=<0 -25 0>, angularTarget=<1 0 0>}
      localJointOffset=<0 0 0>
     }]
    simSpaceSettings {maxAngularAcceleration=50, maxAngularVelocity=8, simSpaceAngularAlpha=0.5}
   }
   Save cached pose 'start cache'(Pose=Component To Local(ComponentPose=@n9)) {cacheName=start cache}
   n10 = "Transform (Modify) Bone - Bone: dyn_scarf_MSTR_L"!(ComponentPose=Local To Component(LocalPose=Use cached pose 'start cache'()), Alpha=1, Rotation="10,0,0") -> (Pose) {
    rotationMode=BMM_Additive
    rotationSpace=BCS_BoneSpace
    alphaBoolBlend {blendOption=Linear}
    boneToModify {boneName=dyn_scarf_MSTR_L}
    rotation {pitch=10}
   }
   n11 = "Transform (Modify) Bone - Bone: dyn_scarf_MSTR_R"!(ComponentPose=@n10, Alpha=1, Rotation="0,20,0") -> (Pose) {
    rotationMode=BMM_Additive
    rotationSpace=BCS_BoneSpace
    alphaBoolBlend {blendOption=Linear}
    boneToModify {boneName=dyn_scarf_MSTR_R}
    rotation {yaw=20}
   }
   n12 = "Constraint - dyn_scarf_02_CST_R (PP)"!(ComponentPose=@n11, Alpha=1, ConstraintWeights_0=0.5, ConstraintWeights_1:float=0.5) -> (Pose) {
    alphaBoolBlend {blendOption=Linear}
    boneToModify {boneName=dyn_scarf_02_CST_R}
    constraintSetup [{targetBone {boneName=clavicle_l}}, {targetBone {boneName=spine_05}}]
    constraintWeights [1,1]
   }
   n13 = "Transform (Modify) Bone - Bone: dyn_scarf_02_R"!(ComponentPose=@n12, Alpha=1, Translation="0,0,0", Rotation="-7.647,6.455,-40.433") -> (Pose) {
    rotationMode=BMM_Additive
    rotationSpace=BCS_BoneSpace
    translationMode=BMM_Additive
    translationSpace=BCS_BoneSpace
    alphaBoolBlend {blendOption=Linear}
    boneToModify {boneName=dyn_scarf_02_R}
    rotation=<rot -7.647 6.455 -40.4331>
   }
   Save cached pose 'backpack cache'(Pose=Component To Local(ComponentPose=@n13)) {cacheName=backpack cache}
  ]}
 EventGraph {nodes [
   Event BlueprintUpdateAnimation()
   n1 = Get bIsSkydiveFloating()
   n2 = Get bIsSkydiveDiveMode()
   > Set gravity overrride(gravity overrride:Vector=SelectVector(A="0,0,500", B="0,0,-980", bPickA=OR Boolean(A=@n1, B=@n2))) -> (Output_Get:Vector)
   n3 = CheckForBackbling()
   > Set isWearingBackpack(isWearingBackpack:bool=OR Boolean(A=@n3, B=@n3.bIsBackblingACape)) -> (Output_Get:bool)
   n4 = GetOwnerSkeletalMeshComponent()
   n5 = BreakTransform(InTransform=GetDeltaTransformFromRefPose(self=@n4, BoneName=thigh_l, BaseName=pelvis))
   > Set thigh l pitch(thigh l pitch:double=MapRangeClamped(Value=BreakRotator(InRot=@n5.Rotation).Pitch, InRangeA=0, InRangeB=30)) -> (Output_Get:double)
   n6 = BreakRotator(InRot=BreakTransform(InTransform=GetDeltaTransformFromRefPose(self=@n4, BoneName=head, BaseName=clavicle_r)).Rotation)
   > Set head yaw(head yaw:double=MapRangeClamped(Value=@n6.Yaw, InRangeA=0, InRangeB=-50, OutRangeA=1, OutRangeB=0)) -> (Output_Get:double)
   > Set head yaw when targeting(head yaw when targeting:double=MapRangeClamped(Value=@n6.Yaw, InRangeA=0, InRangeB=-90)) -> (Output_Get:double)
  ]}
]
variables [
 gravity overrride {type=Vector}
 head yaw {type=number}
 head yaw when targeting {type=number}
 isWearingBackpack {type=boolean}
 thigh l pitch {type=number}
]
