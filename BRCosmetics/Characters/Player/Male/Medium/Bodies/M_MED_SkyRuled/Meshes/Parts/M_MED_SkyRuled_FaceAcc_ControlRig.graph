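# Control Rig for the SkyRuled male face accessory that adjusts the back hair when the character wears a backpack or cape.
# A spherical pose reader measures head orientation, with spine_05 as its parent reference.
# Its output, multiplied separately by the backpack and cape variables, weights two successive rotation offsets on dyn_hair_bk_MSTR, one per worn item.
# Both variables are doubles defaulting to zero, so neither correction applies by default.
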ControlRigBlueprint /BRCosmetics/Characters/Player/Male/Medium/Bodies/M_MED_SkyRuled/Meshes/Parts/M_MED_SkyRuled_FaceAcc_ControlRig
graphs [
 RigVMFunctionLibrary {links [], nodes []}
 RigVMModel {nodes [
   RigUnit_BeginExecution = &RigUnit_BeginExecution()
   n1 = &RigUnit_SphericalPoseReader(
    ActiveRegionScaleFactors=(PositiveWidth=2,NegativeWidth=0.5,PositiveHeight=1.5,NegativeHeight=1)
    ActiveRegionSize=0.15
    Debug=(bDrawDebug=false,bDraw2D=False,bDrawLocalAxes=False,DebugScale=25,DebugSegments=20,DebugThickness=0.25)
    DriverAxis=(X=0,Y=1,Z=0)
    DriverItem=(Type=Bone,Name="head")
    FalloffRegionScaleFactors=(PositiveWidth=1,NegativeWidth=0.5,PositiveHeight=1,NegativeHeight=1)
    FalloffSize=0.3
    FlipHeightScaling=False
    FlipWidthScaling=False
    OptionalParentItem=(Type=Bone,Name="spine_05")
    RotationOffset=(X=90,Y=0,Z=-50)
    ExecutePin=@RigUnit_BeginExecution.ExecutePin
   )
   Multiply = &RigVMFunction_MathDoubleMul(A=@n1.OutputParam, B=RigVMVariableNode() {variable=isWearing Backpack}.Value)
   > &RigUnit_ModifyTransforms(
    ItemToModify=((Item=(Type=Bone,Name="dyn_hair_bk_MSTR"),Transform=(Rotation=(X=0.0739,Y=0.2967,Z=0.0682,W=-0.9497),Translation=(X=0,Y=0,Z=0),Scale3D=(X=1,Y=1,Z=1))))
    WeightMinimum=0
    Weight=@Multiply.Result
   )
   > &RigUnit_ModifyTransforms(
    ItemToModify=((Item=(Type=Bone,Name="dyn_hair_bk_MSTR"),Transform=(Rotation=(X=0.0123,Y=-0.2175,Z=-0.0539,W=0.9745),Translation=(X=0,Y=0,Z=0),Scale3D=(X=1,Y=1,Z=1))))
    WeightMinimum=0
    Weight=&RigVMFunction_MathDoubleMul(A=@n1.OutputParam, B=RigVMVariableNode() {variable=isWearing Cape}.Value).Result
   )
  ]}
]
hierarchy [
 root {loc=<0 0 0>, rot=<0 0 0 1>}
 pelvis {parent=root, loc=<0 -0.7071 93.1191>, rot=<0.0127 0.707 -0.0127 -0.707>}
 spine_01 {parent=pelvis, loc=<6.7494 -0.6388 0>, rot=<0 0 0.034 -0.9994>}
 spine_02 {parent=spine_01, loc=<7.2518 -0.035 0>, rot=<0 0 0.0238 -0.9997>}
 spine_03 {parent=spine_02, loc=<6.0104 0 0>, rot=<0 0 0.0149 -0.9999>}
 spine_04 {parent=spine_03, loc=<8.1024 0 0>, rot=<0 0 -0.0132 -0.9999>}
 spine_05 {parent=spine_04, loc=<10.0146 0 0>, rot=<0 0 -0.0015 -1>}
 neck_01 {parent=spine_05, loc=<11.3138 0.1529 0>, rot=<0 0 -0.2003 -0.9797>}
 neck_02 {parent=neck_01, loc=<4.5627 0 0>, rot=<0 0 0.0504 -0.9987>}
 head {parent=neck_02, loc=<5.3914 0 0>, rot=<0 0 0.115 -0.9934>}
 dyn_hair_fr_MSTR_L {parent=head, loc=<8.677 8.3173 -1.3987>, rot=<-0.1974 -0.6131 -0.1023 -0.7581>}
 dyn_hair_fr_01_L {parent=dyn_hair_fr_MSTR_L, loc=<0 0 0>, rot=<0 0 0 1>}
 dyn_hair_fr_02_L {parent=dyn_hair_fr_01_L, loc=<2.0692 0 0>, rot=<0 -0.605 0 -0.7962>}
 dyn_hair_fr_03_L {parent=dyn_hair_fr_02_L, loc=<2.3203 0 0>, rot=<0 -0.1929 0 -0.9812>}
 dyn_hair_fr_MSTR_R {parent=head, loc=<8.9542 7.5722 2.8691>, rot=<0.7378 0.3148 0.5271 0.2806>}
 dyn_hair_fr_01_R {parent=dyn_hair_fr_MSTR_R, loc=<0 0 0>, rot=<0 0 0 1>}
 dyn_hair_fr_02_R {parent=dyn_hair_fr_01_R, loc=<1.9287 0 0>, rot=<0 -0.5321 0 -0.8467>}
 dyn_hair_fr_03_R {parent=dyn_hair_fr_02_R, loc=<2.2047 0 0>, rot=<0 -0.3662 0 -0.9305>}
 dyn_hair_fr_04_R {parent=dyn_hair_fr_03_R, loc=<3.1296 0 0>, rot=<0 0 0 1>}
 dyn_hair_bk_MSTR {parent=head, loc=<12.0396 -5.6858 0>, rot=<0.6048 -0.3563 0.3252 -0.6337>}
 dyn_hair_bk_01 {parent=dyn_hair_bk_MSTR, loc=<0 0 0>, rot=<0 0 0 1>}
 dyn_hair_bk_02 {parent=dyn_hair_bk_01, loc=<3.6184 0 0>, rot=<0 -0.6626 0 -0.7489>}
 dyn_hair_bk_03 {parent=dyn_hair_bk_02, loc=<4.5463 0 0>, rot=<0 -0.4933 0 -0.8699>}
 dyn_hair_bk_04 {parent=dyn_hair_bk_03, loc=<6.6837 0 0>, rot=<0 0.4352 0 -0.9003>}
 clavicle_l {parent=spine_05, loc=<4.3575 -0.1466 -2.3223>, rot=<0.0465 -0.7289 -0.009 -0.6829>}
 upperarm_l {parent=clavicle_l, loc=<15.5111 0 0>, rot=<0.0226 -0.3234 -0.0088 -0.9459>}
 clavicle_r {parent=spine_05, loc=<4.3575 -0.1466 2.3223>, rot=<0.7289 0.0465 -0.6829 0.009>}
 upperarm_r {parent=clavicle_r, loc=<-15.5111 0 0>, rot=<0.0226 -0.3234 -0.0088 -0.9459>}
]
variables [
 isWearing Backpack {type=double, default=0}
 isWearing Cape {type=double, default=0}
]
