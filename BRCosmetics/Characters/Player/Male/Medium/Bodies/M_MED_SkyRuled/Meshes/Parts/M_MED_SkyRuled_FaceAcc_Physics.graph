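# Physics asset for the SkyRuled male medium body's FaceAcc part, setting up capsule bodies for dynamic hair bones.
# One back chain and asymmetric left and right front chains hang from zero-size master bodies linked to the head.
# Hair segments use tight swing and twist limits, while master and body joints mostly rotate freely.
# Larger capsules on head, neck, upper spine and clavicles approximate the upper body around the hair.
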PhysicsAsset /BRCosmetics/Characters/Player/Male/Medium/Bodies/M_MED_SkyRuled/Meshes/Parts/M_MED_SkyRuled_FaceAcc_Physics
bodies {
 clavicle_l [{shapeType=Capsule, length=10, radius=10, shapeName=clavicle_l_capsule, center=<5 0 0>, rotation=<rot -0.0002 -90 90>}]
 clavicle_r [{shapeType=Capsule, length=10, radius=10, shapeName=clavicle_r_capsule, center=<-5 0 0>, rotation=<rot -0.0002 -90 90>}]
 dyn_hair_bk_01 [{shapeType=Capsule, length=1.5, radius=2.5, shapeName=dyn_hair_bk_01_capsule, center=<3.2412 0 1.1278>, rotation=<rot 90 0.0001 0>}]
 dyn_hair_bk_02 [{shapeType=Capsule, length=3, radius=3, shapeName=dyn_hair_bk_02_capsule, center=<3.2412 0 0>, rotation=<rot 90 0.0002 0>}]
 dyn_hair_bk_03 [{shapeType=Capsule, length=4, radius=3, shapeName=dyn_hair_bk_03_capsule, center=<3.2412 0 0>, rotation=<rot 69.9998 0 0>}]
 dyn_hair_bk_04 [{shapeType=Capsule, length=5.5, radius=3, shapeName=dyn_hair_bk_04_capsule, center=<3.2412 0 0>, rotation=<rot 80 -179.9998 -180>}]
 dyn_hair_bk_MSTR [{shapeType=Capsule, length=0, radius=0, shapeName=dyn_hair_bk_MSTR_capsule}]
 dyn_hair_fr_01_L [{shapeType=Capsule, length=1, radius=1, shapeName=dyn_hair_fr_01_L_capsule, center=<1.5214 0 0>, rotation=<rot -90 -0.0006 0>}]
 dyn_hair_fr_01_R [{shapeType=Capsule, length=1, radius=1, shapeName=dyn_hair_fr_01_R_capsule, center=<1.5214 0 0>, rotation=<rot -90 0.0005 0>}]
 dyn_hair_fr_02_L [{shapeType=Capsule, length=1, radius=1, shapeName=dyn_hair_fr_02_L_capsule, center=<1.5214 0 0>, rotation=<rot -90 0.0002 0>}]
 dyn_hair_fr_02_R [{shapeType=Capsule, length=1, radius=1, shapeName=dyn_hair_fr_02_R_capsule, center=<1.5214 0 0>, rotation=<rot -90 0.0001 0>}]
 dyn_hair_fr_03_L [{shapeType=Capsule, length=1, radius=1, shapeName=dyn_hair_fr_03_L_capsule, center=<1.5214 0 0>, rotation=<rot -90 0.0004 0>}]
 dyn_hair_fr_03_R [{shapeType=Capsule, length=1, radius=1, shapeName=dyn_hair_fr_03_R_capsule, center=<1.5214 0 0>, rotation=<rot -90 -0.0003 0>}]
 dyn_hair_fr_04_R [{shapeType=Capsule, length=1, radius=1, shapeName=dyn_hair_fr_04_R_capsule, center=<1.5214 0 0>, rotation=<rot -90 -0.0003 0>}]
 dyn_hair_fr_MSTR_L [{shapeType=Capsule, length=0, radius=0, shapeName=dyn_hair_fr_MSTR_L_capsule}]
 dyn_hair_fr_MSTR_R [{shapeType=Capsule, length=0, radius=0, shapeName=dyn_hair_fr_MSTR_R_capsule}]
 head [{shapeType=Capsule, length=5.9462, radius=8.045, shapeName=head_capsule, center=<4.384 2.1081 0>, rotation=<rot -90 -0.0005 0>}]
 neck_01 [{shapeType=Capsule, length=0, radius=9.005, shapeName=neck_01_capsule, rotation=<rot 0 -90.0002 90>}]
 spine_05 [{shapeType=Capsule, length=10.505, radius=13.505, shapeName=spine_05_capsule, rotation=<rot 0 -90 0>}]
}
constraints [
 {
  bone1Name=dyn_hair_bk_MSTR
  bone2Name=head
  swing1LimitDegrees=45
  swing1Motion=Free
  swing2LimitDegrees=45
  swing2Motion=Free
  twistLimitDegrees=45
  twistMotion=Free
 }
 {bone1Name=dyn_hair_bk_01, bone2Name=dyn_hair_bk_MSTR, swing1LimitDegrees=0, swing2LimitDegrees=10, twistLimitDegrees=5}
 {bone1Name=dyn_hair_bk_02, bone2Name=dyn_hair_bk_01, swing1LimitDegrees=10, swing2LimitDegrees=45, twistLimitDegrees=5}
 {bone1Name=dyn_hair_bk_03, bone2Name=dyn_hair_bk_02, swing1LimitDegrees=10, swing2LimitDegrees=25, twistLimitDegrees=5}
 {bone1Name=dyn_hair_bk_04, bone2Name=dyn_hair_bk_03, swing1LimitDegrees=10, swing2LimitDegrees=15, twistLimitDegrees=5}
 {
  bone1Name=dyn_hair_fr_MSTR_L
  bone2Name=head
  swing1LimitDegrees=45
  swing1Motion=Free
  swing2LimitDegrees=45
  swing2Motion=Free
  twistLimitDegrees=45
  twistMotion=Free
 }
 {bone1Name=dyn_hair_fr_01_L, bone2Name=dyn_hair_fr_MSTR_L, swing1LimitDegrees=0, swing2LimitDegrees=10, twistLimitDegrees=5}
 {bone1Name=dyn_hair_fr_02_L, bone2Name=dyn_hair_fr_01_L, swing1LimitDegrees=0, swing2LimitDegrees=10, twistLimitDegrees=5}
 {bone1Name=dyn_hair_fr_03_L, bone2Name=dyn_hair_fr_02_L, swing1LimitDegrees=10, swing2LimitDegrees=10, twistLimitDegrees=5}
 {
  bone1Name=dyn_hair_fr_MSTR_R
  bone2Name=head
  swing1LimitDegrees=45
  swing1Motion=Free
  swing2LimitDegrees=45
  swing2Motion=Free
  twistLimitDegrees=45
  twistMotion=Free
 }
 {bone1Name=dyn_hair_fr_01_R, bone2Name=dyn_hair_fr_MSTR_R, swing1LimitDegrees=0, swing2LimitDegrees=10, twistLimitDegrees=5}
 {bone1Name=dyn_hair_fr_02_R, bone2Name=dyn_hair_fr_01_R, swing1LimitDegrees=0, swing2LimitDegrees=10, twistLimitDegrees=5}
 {bone1Name=dyn_hair_fr_03_R, bone2Name=dyn_hair_fr_02_R, swing1LimitDegrees=10, swing2LimitDegrees=10, twistLimitDegrees=5}
 {bone1Name=dyn_hair_fr_04_R, bone2Name=dyn_hair_fr_03_R, swing1LimitDegrees=10, swing2LimitDegrees=45, twistLimitDegrees=5}
 {
  bone1Name=head
  bone2Name=neck_01
  swing1LimitDegrees=45
  swing1Motion=Free
  swing2LimitDegrees=45
  swing2Motion=Free
  twistLimitDegrees=45
  twistMotion=Free
 }
 {
  bone1Name=neck_01
  bone2Name=spine_05
  swing1LimitDegrees=45
  swing1Motion=Free
  swing2LimitDegrees=45
  swing2Motion=Free
  twistLimitDegrees=45
  twistMotion=Free
 }
 {
  bone1Name=clavicle_l
  bone2Name=spine_05
  swing1LimitDegrees=45
  swing1Motion=Free
  swing2LimitDegrees=45
  swing2Motion=Free
  twistLimitDegrees=45
  twistMotion=Free
 }
 {
  bone1Name=clavicle_r
  bone2Name=spine_05
  swing1LimitDegrees=45
  swing1Motion=Free
  swing2LimitDegrees=45
  swing2Motion=Free
  twistLimitDegrees=45
  twistMotion=Free
 }
]
