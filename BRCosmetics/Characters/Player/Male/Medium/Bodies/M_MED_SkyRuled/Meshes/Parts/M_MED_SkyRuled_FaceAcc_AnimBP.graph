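# Animation Blueprint for the SkyRuled male character's face accessory part, copying the attached parent mesh's pose and adding neck-relative RigidBody secondary motion.
# A backbling check selects either plain physics or slightly differently tuned physics followed by a Control Rig.
# The update event also computes cape/backpack weights, skydive gravity and head-yaw falloff values.
# A back-hair Modify Bone chain and an extra RigidBody node do not feed the output.
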AnimBlueprint /BRCosmetics/Characters/Player/Male/Medium/Bodies/M_MED_SkyRuled/Meshes/Parts/M_MED_SkyRuled_FaceAcc_AnimBP
target_skeleton=&/M_MED_SkyRuled_FaceAcc_Skeleton
graphs [
 AnimGraph {nodes [
   n1 = "Transform (Modify) Bone - Bone: dyn_hair_bk_01"(Alpha=1, Translation="0,0,0", Rotation="-10,0,0", Scale="1,1,1") {
    rotationMode=BMM_Additive
    rotationSpace=BCS_BoneSpace
    translationMode=BMM_Additive
    translationSpace=BCS_BoneSpace
    alphaBoolBlend {blendOption=Linear}
    boneToModify {boneName=dyn_hair_bk_01}
    rotation {pitch=-10}
   }
   n2 = "Transform (Modify) Bone - Bone: dyn_hair_bk_02"(ComponentPose=@n1, Alpha=1, Translation="0,0,0", Rotation="-20,0.004,0.001", Scale="1,1,1") {
    rotationMode=BMM_Additive
    rotationSpace=BCS_BoneSpace
    translationMode=BMM_Additive
    translationSpace=BCS_BoneSpace
    alphaBoolBlend {blendOption=Linear}
    boneToModify {boneName=dyn_hair_bk_02}
    rotation=<rot -20 0.004 0.001>
   }
   n3 = "Transform (Modify) Bone - Bone: dyn_hair_bk_03"(ComponentPose=@n2, Alpha=1, Translation="0,0,0", Rotation="60.000,0.002,0.003", Scale="1,1,1") {
    rotationMode=BMM_Additive
    rotationSpace=BCS_BoneSpace
    translationMode=BMM_Additive
    translationSpace=BCS_BoneSpace
    alphaBoolBlend {blendOption=Linear}
    boneToModify {boneName=dyn_hair_bk_03}
    rotation=<rot 59.9999 0.0021 0.0031>
   }
   "Transform (Modify) Bone - Bone: dyn_hair_bk_04"(ComponentPose=@n3, Alpha=1, Translation="0,0,0", Rotation="-30.000,-0.004,0.026", Scale="1,1,1") {
    rotationMode=BMM_Additive
    rotationSpace=BCS_BoneSpace
    translationMode=BMM_Additive
    translationSpace=BCS_BoneSpace
    alphaBoolBlend {blendOption=Linear}
    boneToModify {boneName=dyn_hair_bk_04}
    rotation=<rot -30.0001 -0.0039 0.026>
   }
   RigidBody!(ComponentPose, Alpha, OverrideWorldGravity:Vector="0,0,-980") -> (Pose) {
    bClampLinearTranslationLimitToRefPose=true
    bOverrideWorldGravity=true
    overridePhysicsAsset=&/M_MED_SkyRuled_FaceAcc_Physics
    simulationSpace=BaseBoneSpace
    alphaBoolBlend {blendOption=Linear}
    baseBoneRef {boneName=neck_01}
    simSpaceSettings {
     dampingAlpha=0
     maxAngularAcceleration=7
     maxAngularVelocity=5
     maxLinearAcceleration=1000
     maxLinearVelocity=700
     velocityScaleZ=0.2
     worldAlpha=0.2
     externalLinearDragV=<1 1 1>
    }
   }
   n4 = RigidBody(ComponentPose=Local To Component(LocalPose=Use cached pose 'start cache'()), OverrideWorldGravity="0,0,-980") {
    bClampLinearTranslationLimitToRefPose=true
    bOverrideWorldGravity=true
    overridePhysicsAsset=&M_MED_SkyRuled_FaceAcc_Physics
    simulationSpace=BaseBoneSpace
    alphaBoolBlend {blendOption=Linear}
    baseBoneRef {boneName=neck_01}
    simSpaceSettings {
     dampingAlpha=0
     maxAngularAcceleration=7
     maxAngularVelocity=7
     maxLinearAcceleration=1000
     maxLinearVelocity=700
     velocityScaleZ=0.33
     worldAlpha=1
     externalLinearDragV=<1 1 1>
    }
   }
   Output Pose(Result=Blend Poses by bool(
     BlendPose_0=Use cached pose 'backpack cache'()
     BlendPose_1=Component To Local(ComponentPose=@n4)
     bActiveValue=Get isWearing Cape/Backpack() -> (isWearing Cape/Backpack:bool)
    ) {blendType=Linear}) {layerGroup=None}
   Save cached pose 'start cache'(Pose=Copy Pose From Mesh() {bUseAttachedParent=true}) {cacheName=start cache}
   n5 = RigidBody(ComponentPose=Local To Component(LocalPose=Use cached pose 'start cache'()), OverrideWorldGravity="0,0,-980") {
    bClampLinearTranslationLimitToRefPose=true
    bOverrideWorldGravity=true
    overridePhysicsAsset=&M_MED_SkyRuled_FaceAcc_Physics
    simulationSpace=BaseBoneSpace
    alphaBoolBlend {blendOption=Linear}
    baseBoneRef {boneName=neck_01}
    simSpaceSettings {
     dampingAlpha=0
     maxAngularAcceleration=5
     maxAngularVelocity=5
     maxLinearAcceleration=1000
     maxLinearVelocity=700
     velocityScaleZ=0.2
     worldAlpha=1
     externalLinearDragV=<1 1 1>
    }
   }
   n6 = Control Rig!(Source=Component To Local(ComponentPose=@n5), Alpha, isWearing Backpack:double=0, isWearing Cape:double=0) -> (Pose) {
    alphaBoolBlend {blendOption=Linear}
    controlRigAssetReference {blueprintRigClass=&/M_MED_SkyRuled_FaceAcc_ControlRig.M_MED_SkyRuled_FaceAcc_ControlRig_C}
   }
   Save cached pose 'backpack cache'(Pose=@n6) {cacheName=backpack cache}
  ]}
 EventGraph {nodes [
   Event BlueprintUpdateAnimation()
   > n1 = Sequence()
   n2 = CheckForBackbling()
   > Set isWearing Cape/Backpack(isWearing Cape/Backpack:bool=OR Boolean(A=@n2, B=@n2.bIsBackblingACape)) -> (Output_Get:bool)
   n3 = NOT Boolean(A=@n2.bIsBackblingACape)
   > Set isWearing Backpack(isWearing Backpack:double=SelectFloat(A=1, B=0, bPickA=AND Boolean(A=@n2, B=@n3))) -> (Output_Get:double)
   > Set isWearing Cape(isWearing Cape:double=SelectFloat(A=1, B=0, bPickA=@n2.bIsBackblingACape)) -> (Output_Get:double)
   n4 = Get bIsSkydiveFloating()
   n5 = Get bIsSkydiveDiveMode()
   > Set gravity overrride(gravity overrride:Vector=SelectVector(A="0,0,500", B="0,0,-980", bPickA=OR Boolean(A=@n4, B=@n5))) -> (Output_Get:Vector)
   n6 = GetOwnerSkeletalMeshComponent()
   n7 = BreakRotator(InRot=BreakTransform(InTransform=GetDeltaTransformFromRefPose(self=@n6, BoneName=head, BaseName=spine_05)).Rotation)
   Set head yaw(execute=@n1.then_1, head yaw:float=MapRangeClamped(Value=@n7.Yaw, InRangeA=-10, InRangeB=-40)) -> (Output_Get:float)
   > Set turn off physics(turn off physics:double=MapRangeClamped(Value=@n7.Yaw, InRangeA=-10, InRangeB=-40, OutRangeA=1, OutRangeB=0)) -> (Output_Get:double)
   n8 = Event OnCharacterCustomizationLoaded()
   > Set BodyMesh(BodyMesh=Get Array Item with EFort Custom Part Type(Array=@n8.CustomMeshParts))
  ]}
]
variables [
 BodyMesh {type=SkeletalMeshComponent}
 gravity overrride {type=Vector}
 head yaw {type=number}
 isWearing Backpack {type=number}
 isWearing Cape {type=number}
 isWearing Cape/Backpack {type=boolean}
 turn off physics {type=number}
]
